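# Common UI tab button showing an icon and optional text label inside a bang (notification) wrapper.
# Tab label info or PreConstruct feeds Set Icon and Set Text, which collapse empty content; shown text adds an icon tooltip.
# Hover and selection events tint the icon, optionally rebuilding its brush with selected/unselected colors.
# Construct fixes height for non-gamepad mobile input; hover sound plays only for two hardcoded label strings when unselected.
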WidgetBlueprint /FortUILibrary/Foundation/Buttons/IconTabButton_Leto
graphs [
 EventGraph {nodes [
   Event On Hovered()
   > SetColorAndOpacity(
    self:Image Object Reference=Get LeftSideImage() -> (LeftSideImage:Image Object Reference)
    InColorAndOpacity:Linear Color Structure=Get HoveredIconTint() -> (HoveredIconTint:Linear Color Structure)
   )
   > Play Hover Sound()
   n1 = Event SetTabLabelInfo!() -> (OutputDelegate, then, TabLabelInfo:"Fort Tab Button Label Info Structure (by ref)") {is=title K2Node_Event}
   > n2 = Sequence()
   n3 = Break Fort Tab Button Label Info!(FortTabButtonLabelInfo:"Fort Tab Button Label Info Structure (by ref)"=@n1.TabLabelInfo) -> (DisplayName:Text, IconBrush:Slate Brush Structure, IconResource:Soft Object Reference) {is=title K2Node_BreakStruct}
   > Set Text!(execute, self, ButtonText:Text=@n3) -> (then) {is=title K2Node_CallFunction}
   Set Icon!(execute=@n2.then_1, self, IconBrush:Slate Brush Structure=@n3.IconBrush) -> (then) {is=title K2Node_CallFunction}
   Event OnCurrentTextStyleChanged() {is=title K2Node_Event}
   > SetStyle!(
    execute
    self:Common Text Object Reference=Get CenterButtonTextWidget() -> (CenterButtonTextWidget:Common Text Object Reference)
    InStyle:Common Text Style Class Reference=GetCurrentTextStyleClass!(self:Common Button Base Object Reference) -> (ReturnValue:Common Text Style Class Reference)
   ) -> (then)
   Event On Unhovered()
   > n4 = Branch(Condition=GetSelected!(self:Common Button Base Object Reference) -> (ReturnValue:bool))
   n5 = Get LeftSideImage()
   > SetColorAndOpacity(self=@n5, InColorAndOpacity=Get HoveredIconTint())
   SetColorAndOpacity(execute=@n4.else, self=@n5, InColorAndOpacity=Get DeselectedIconTint() -> (DeselectedIconTint:Linear Color Structure))
   Event Construct()
   n6 = GetCurrentInputType(self=Get CommonInputSubsystem!() -> (ReturnValue:Common Input Object Reference) {is=title K2Node_GetBlueprintContextEx})
   > n7 = Branch(Condition=AND Boolean(A=IsMobileGame(), B="Not Equal (Enum)"!(A:ECommonInputType Enum=@n6, B:ECommonInputType Enum=Gamepad) -> (ReturnValue:bool)))
   n8 = Get SizeBoxShell() -> (SizeBoxShell:Size Box Object Reference)
   > n9 = SetHeightOverride(self=@n8, InHeightOverride=64)
   Set Bang State!(
    execute=[@n9.then, @n12.then]
    self:Fort Bang Wrapper NUI Object Reference=Get BangWrapper() -> (BangWrapper:Normal Bang Wrapper Object Reference)
    bEnabled:bool=false
    Count:int=0
   ) -> (then) {is=title K2Node_CallFunction}
   > n10 = Branch(Condition=GetSelected())
   n11 = Get LeftSideImage()
   > SetColorAndOpacity(self=@n11, InColorAndOpacity=Get HoveredIconTint())
   SetColorAndOpacity(execute=@n10.else, self=@n11, InColorAndOpacity=Get DeselectedIconTint())
   n12 = ClearHeightOverride(execute=@n7.else, self:Size Box Object Reference=@n8)
   Event PreConstruct()
   > n13 = Sequence()
   Set Text(execute=@n13.then_1, ButtonText=Get ButtonText() -> (ButtonText:Text))
   > SetStyle(self=Get CenterButtonTextWidget(), InStyle=GetCurrentTextStyleClass())
   > Adjust Bang Image Translation!(execute, self:Normal Bang Wrapper Object Reference=Get BangWrapper(), Translation:Vector 2D Structure=(X=25,Y=0)) -> (then)
   Set Icon(execute=@n13.then_0, IconBrush=Get IconBrush() -> (IconBrush:Slate Brush Structure))
   Event On Enabled() {is=title K2Node_Event}
   > Update Bang State!(execute, self, bBangEnabled:bool=Get bBangEnabled() -> (bBangEnabled:bool), Count:int=0) -> (then)
   Event On Disabled() {is=title K2Node_Event}
   > Update Bang State(bBangEnabled=Get bBangEnabled(), Count=0)
   Event On Selected()
   > SetColorAndOpacity(self=Get LeftSideImage(), InColorAndOpacity=Get SelectedIconTint() -> (SelectedIconTint:Linear Color Structure))
   > PlaySound2D()
   > Branch(Condition=Get ChangeIconColorWhenSelected() -> (ChangeIconColorWhenSelected:bool))
   n14 = Break Slate Brush!(SlateBrush:"Slate Brush Structure (by ref)"=Get Brush!(self:Image Object Reference=Get LeftSideImage()) -> (Brush:Slate Brush Structure)) -> (
    TintColor:Slate Color Structure
    DrawAs:ESlateBrushDrawType Enum
    Tiling:ESlateBrushTileType Enum
    Mirroring:ESlateBrushMirrorType Enum
    ImageSize:Vector 2D Structure
    Margin:Margin Structure
    ResourceObject:Object Reference
    OutlineSettings:Slate Brush Outline Settings Structure
   ) {is=title K2Node_BreakStruct}
   > SetBrush(
    self=Get LeftSideImage()
    InBrush=Make Slate Brush!(
     TintColor:Slate Color Structure=Get SelectedIconColor() -> (SelectedIconColor:Slate Color Structure)
     DrawAs:ESlateBrushDrawType Enum=@n14.DrawAs
     Tiling:ESlateBrushTileType Enum=@n14.Tiling
     Mirroring:ESlateBrushMirrorType Enum=@n14.Mirroring
     ImageSize:Vector 2D Structure=@n14.ImageSize
     Margin:Margin Structure=@n14.Margin
     ResourceObject:Object Reference=@n14.ResourceObject
     OutlineSettings:Slate Brush Outline Settings Structure=(CornerRadii=(X=0,Y=0,Z=0,W=0),Color=(SpecifiedColor=(R=0,G=0,B=0,A=0),ColorUseRule=UseColor_Specified),Width=0,RoundingType=HalfHeightRadius,bUseBrushTransparency=False)
    ) -> (SlateBrush:Slate Brush Structure) {is=title K2Node_MakeStruct}
   )
   Event On Deselected()
   > SetColorAndOpacity(self=Get LeftSideImage(), InColorAndOpacity=Get DeselectedIconTint())
   > Branch(Condition=Get ChangeIconColorWhenSelected())
   n15 = Break Slate Brush(SlateBrush=Get Brush(self=Get LeftSideImage()))
   > SetBrush(
    self=Get LeftSideImage()
    InBrush=Make Slate Brush(
     TintColor=Get UnSelectedIconColor() -> (UnSelectedIconColor:Slate Color Structure)
     DrawAs=@n15.DrawAs
     Tiling=@n15.Tiling
     Mirroring=@n15.Mirroring
     ImageSize=@n15.ImageSize
     Margin=@n15.Margin
     ResourceObject=@n15.ResourceObject
     OutlineSettings=(CornerRadii=(X=0,Y=0,Z=0,W=0),Color=(SpecifiedColor=(R=0,G=0,B=0,A=0),ColorUseRule=UseColor_Specified),Width=0,RoundingType=HalfHeightRadius,bUseBrushTransparency=False)
    )
   )
  ]}
 Play Hover Sound {nodes [
   Play Hover Sound() {is=title K2Node_FunctionEntry}
   n1 = GetSelected()
   n2 = Get ButtonText()
   n3 = "Equal Exactly (Text)"!(A:"Text (by ref)"=@n2, B:"Text (by ref)"="To Text (String)"(InString=EDIT STYLE)) -> (ReturnValue:bool)
   n4 = "Equal Exactly (Text)"(A=@n2, B="To Text (String)"(InString=SELECT OUTFIT))
   > Branch(Condition=AND Boolean(A=OR Boolean(A=@n4, B=@n3), B=NOT Boolean(A=@n1)))
   > PlaySound2D(Sound=/Game/Sounds/Fort_UI_Sounds/FrontEnd/DefaultSounds/PressedButtons/MSS_UI_Button_Hover_Default)
  ]}
 Set Icon {nodes [
   n1 = Set Icon!() -> (then, IconBrush:Slate Brush Structure) {is=title K2Node_FunctionEntry}
   > Set IconBrush(IconBrush:Slate Brush Structure=@n1) -> (Output_Get:Slate Brush Structure)
   > SetBrush(self=Get LeftSideImage(), InBrush=Get IconBrush())
   n2 = Break Slate Brush(SlateBrush=Get IconBrush())
   > SetVisibility(
    self=Get LeftSideImage()
    InVisibility=Select(Option 0=Visible, Option 1=Collapsed, Index="Equal (Enum)"(A:ESlateBrushDrawType Enum=@n2.DrawAs, B:ESlateBrushDrawType Enum=NoDrawType))
   )
  ]}
 Set Text {nodes [
   n1 = Set Text!() -> (then, ButtonText:Text) {is=title K2Node_FunctionEntry}
   > n2 = Set ButtonText(ButtonText:Text=@n1) -> (Output_Get:Text)
   n3 = TextIsEmpty(InText=@n1)
   > n4 = Branch(Condition=AND Boolean(A=NOT Boolean(A=@n3), B=Get UseText() -> (UseText:bool)))
   n5 = Get CenterButtonTextWidget()
   > SetVisibility(self=@n5, InVisibility=SelfHitTestInvisible)
   > "SetText (Text)"(self=@n5, InText=@n2)
   > n6 = Create Basic Tooltip!(
    execute
    Owning Player:Player Controller Object Reference=GetOwningPlayer()
    Body Text:Text
    Header Text:Text=Get ButtonText()
    Icon Brush:Slate Brush Asset Object Reference
    Header Style:Common Text Style Class Reference
    Body Style:Common Text Style Class Reference
   ) -> (then, Output:User Widget Object Reference)
   > SetToolTip!(execute, self:Widget Object Reference=Get LeftSideImage(), Widget:Widget Object Reference=@n6) -> (then)
   SetVisibility(execute=@n4.else, self=@n5, InVisibility=Collapsed)
  ]}
 SetChangeIconColorWhenSelected {nodes [
   SetChangeIconColorWhenSelected!() -> (then, ChangeColorWhenSelected:bool=~, SelectedColor:Slate Color Structure, UnselectedColor:Slate Color Structure) {is=title K2Node_FunctionEntry}
   > Set ChangeIconColorWhenSelected(ChangeIconColorWhenSelected:bool=Get ChangeColorWhenSelected() -> (ChangeColorWhenSelected:bool)) -> (Output_Get:bool)
   > Set SelectedIconColor(SelectedIconColor:Slate Color Structure=Get SelectedColor() -> (SelectedColor:Slate Color Structure)) -> (Output_Get:Slate Color Structure)
   > Set UnSelectedIconColor(UnSelectedIconColor:Slate Color Structure=Get UnselectedColor() -> (UnselectedColor:Slate Color Structure)) -> (Output_Get:Slate Color Structure)
   > Branch(Condition=Get ChangeIconColorWhenSelected())
   > SetBrushTintColor!(
    execute
    self:Image Object Reference=Get LeftSideImage()
    TintColor:Slate Color Structure=Select(Option 0:Slate Color Structure=Get UnSelectedIconColor(), Option 1:Slate Color Structure=Get SelectedIconColor(), Index=GetSelected()) -> (ReturnValue:Slate Color Structure)
   ) -> (then)
  ]}
 SetTutorialBorderStyle {nodes [
   n1 = SetTutorialBorderStyle!() -> (then, BorderStyle:Common Border Style Class Reference) {is=title K2Node_FunctionEntry}
   > SetTutorialBorderStyle!(execute, self:Normal Bang Wrapper Object Reference=Get BangWrapper(), BorderStyle:Common Border Style Class Reference=@n1) -> (then) {is=title K2Node_CallFunction}
  ]}
 ShowText {nodes [
   ShowText() {is=title K2Node_FunctionEntry}
   > Set UseText(UseText:bool=true) -> (Output_Get:bool)
   > Set Text(ButtonText=Get ButtonText()) {is=title K2Node_CallFunction}
  ]}
 Update Bang State {nodes [
   n1 = Update Bang State!() -> (then, bBangEnabled:bool=~, Count:int) {is=title K2Node_FunctionEntry}
   > Set bBangEnabled(bBangEnabled:bool=@n1) -> (Output_Get:bool)
   > Set Bang State(
    self=Get BangWrapper()
    bEnabled=AND Boolean!(A=Get bBangEnabled(), B=GetIsEnabled!(self:Widget Object Reference) -> (ReturnValue:bool), C:bool=Get bIsEnabled() -> (bIsEnabled:bool)) -> (ReturnValue)
    Count=@n1.Count
   )
  ]}
]
variables [
 bBangEnabled {type=boolean}
 ButtonText {type=string, default="{Text}"}
 ChangeIconColorWhenSelected {type=boolean}
 DeselectedIconTint {type=LinearColor, default=<rgba 1 1 1 0.4>}
 HoveredIconTint {type=LinearColor, default=<rgba 1 1 1 1>}
 IconBrush {
  type=SlateBrush
  default {
   bIsDynamicallyLoaded=false
   drawAs=NoDrawType
   imageType=NoImage
   mirroring=NoMirror
   resourceName=None
   resourceObject=null
   tiling=NoTile
   imageSize=<0 0>
   margin=<ltrb 0 0 0 0>
   outlineSettings {
    bUseBrushTransparency=false
    roundingType=HalfHeightRadius
    width=0
    color {colorUseRule=UseColor_Specified, specifiedColor=<rgba 0 0 0 0>}
    cornerRadii=<0 0 0 1>
   }
   tintColor {colorUseRule=UseColor_Specified, specifiedColor=<rgba 1 1 1 1>}
   uVRegion {bIsValid=false, max=<0 0>, min=<0 0>}
  }
 }
 SelectedIconColor {type=SlateColor, default {colorUseRule=UseColor_Specified, specifiedColor=<rgba 0 0 0 1>}}
 SelectedIconTint {type=LinearColor, default=<rgba 1 1 1 1>}
 UnSelectedIconColor {type=SlateColor, default {colorUseRule=UseColor_Specified, specifiedColor=<rgba 1 1 1 1>}}
 UseText {type=boolean}
]
widgets {
 info {inheritedWidgetCount=1, namedSlotCount=1, parentClass=&CommonButtonBase, rootWidgetClass=&SizeBox, widgetCount=7}
 widgets [
  SizeBoxShell {widgetClassPath=&SizeBox, bIsVariable=true, widget=SizeBoxShell, properties {heightOverride=300}}
  Overlay_1 {widgetClassPath=&Overlay, parent=SizeBoxShell, widget=Overlay_1, slot {class=SizeBoxSlot}}
  BangWrapper {
   widgetClassPath=/Game/UI/Foundation/Bangs/NormalBangWrapper
   bIsVariable=true
   parent=Overlay_1
   widget=BangWrapper
   properties {bangType=Custom, useSmallVersion=true}
   slot {class=OverlaySlot, properties {horizontalAlignment=HAlign_Center, verticalAlignment=VAlign_Fill}}
  }
  ContentHB {widgetClassPath=&HorizontalBox, bIsVariable=true, namedSlotHost=BangWrapper, widget=ContentHB}
  LeftSideImage {
   widgetClassPath=&Image
   bIsVariable=true
   parent=ContentHB
   widget=LeftSideImage
   properties {brush {
     bIsDynamicallyLoaded=false
     drawAs=NoDrawType
     imageType=NoImage
     mirroring=NoMirror
     resourceName=None
     resourceObject=null
     tiling=NoTile
     imageSize=<0 0>
     margin=<ltrb 0 0 0 0>
     outlineSettings {
      bUseBrushTransparency=false
      roundingType=HalfHeightRadius
      width=0
      color {colorUseRule=UseColor_Specified, specifiedColor=<rgba 0 0 0 0>}
      cornerRadii=<0 0 0 0>
     }
     tintColor {colorUseRule=UseColor_Specified, specifiedColor=<rgba 1 1 1 1>}
     uVRegion {bIsValid=false, max=<0 0>, min=<0 0>}
    }}
   slot {class=HorizontalBoxSlot, properties {verticalAlignment=VAlign_Center, padding=<ltrb 4 4 4 4>}}
  }
  CenterButtonTextWidget {
   widgetClassPath=&CommonTextBlock
   bIsVariable=true
   parent=ContentHB
   widget=CenterButtonTextWidget
   properties {
    justification=Center
    style=&TextStyle-Base_C
    text="{Text}"
    visibility=Collapsed
    font {
     bForceMonospaced=false
     bMaterialIsStencil=false
     fontMaterial=null
     fontObject=&NotoSans
     letterSpacing=0
     monospacedWidth=1
     size=24
     skewAmount=0
     typefaceFontName=NotoSans-Regular
     outlineSettings {
      bApplyOutlineToDropShadows=false
      bMiteredCorners=false
      bSeparateFillAlpha=false
      outlineMaterial=null
      outlineSize=0
      outlineColor=<rgba 0 0 0 1>
     }
    }
    shadowOffset=<0 0>
   }
   slot {class=HorizontalBoxSlot, properties {verticalAlignment=VAlign_Center, padding=<ltrb 4 4 4 4>, size {sizeRule=Fill, value=1}}}
  }
  InputActionWidget {widgetClassPath=&CommonActionWidget, bInherited=true, widget=null}
 ]
}
