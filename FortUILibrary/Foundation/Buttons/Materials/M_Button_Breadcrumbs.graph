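# Translucent UI material for breadcrumb buttons, drawing a procedural two-tone dot from an outer and an inner radial gradient.
# Two RadialGradientExponential calls share UVs, each with its own radius and density parameters.
# The inner gradient blends RGBOuter into RGBInner for color.
# A separate alpha blend starts from the AlphaOuter-scaled outer gradient and rises toward one by the AlphaInner-scaled inner gradient.
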Material /FortUILibrary/Foundation/Buttons/Materials/M_Button_Breadcrumbs
graph {
 properties {
  bUsedWithStaticMesh=false
  blendMode=BLEND_Translucent
  materialDomain=MD_UI
  pixelDepthOffsetMode=PDOM_Legacy
  usedShadingModels=Default Lit
 }
 nodes [
  n1 = TextureCoordinate()
  n2 = ScalarParameter() {defaultValue=0.5, parameterName=RadiusOuter}
  n3 = ScalarParameter() {defaultValue=32, parameterName=DensityOuter}
  n4 = MaterialFunctionCall(
   UVs=@n1
   Radius=ScalarParameter() {defaultValue=0.4, parameterName=RadiusInner}
   Density=ScalarParameter() {defaultValue=32, parameterName=DensityInner}
  ) {materialFunction=&RadialGradientExponential}
  LinearInterpolate(
   A=Multiply(
    A=MaterialFunctionCall(UVs=@n1, Radius=@n2, Density=@n3) {materialFunction=&RadialGradientExponential}.RadialGradientExponential
    B=ScalarParameter() {defaultValue=0.5, parameterName=AlphaOuter}
   )
   Alpha=Multiply(A=@n4.RadialGradientExponential, B=ScalarParameter() {defaultValue=1, parameterName=AlphaInner})
  )
  LinearInterpolate(
   A=VectorParameter() {parameterName=RGBOuter, defaultValue=<rgba 0.0462 0.0208 1 1>}.RGB
   B=VectorParameter() {parameterName=RGBInner, defaultValue=<rgba 1 0.1879 0.1719 1>}.RGB
   Alpha=@n4.RadialGradientExponential
  )
 ]
}
