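# Translucent UI material that draws a skewed button shape from a distance-field texture, with a hover tint.
# Center-scaled UVs sample the texture, and AAThreshold cuts an antialiased mask at ShapeSize.
# That mask drives the color, a blend of ShapeColor toward HoverColor by HoverAmount.
# A second threshold, offset by OuterEdgeSize and weighted by InnerOpacity, is subtracted from the mask to set the edge apart.
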Material /FortUILibrary/Foundation/Buttons/Materials/M_SkewButtonShape_DF1
graph {
 properties {
  bUsedWithStaticMesh=false
  blendMode=BLEND_Translucent
  materialDomain=MD_UI
  pixelDepthOffsetMode=PDOM_Legacy
  usedShadingModels=Default Lit
 }
 nodes [
  n1 = ScalarParameter() {defaultValue=0.9, parameterName=ShapeSize}
  n2 = MaterialFunctionCall(Texture Scale=AppendVector(A=ScalarParameter() {defaultValue=1, parameterName=XScale}, B=ScalarParameter() {defaultValue=1, parameterName=YScale})) {materialFunction=&ScaleUVsByCenter}
  n3 = TextureSampleParameter2D(UVs=@n2.UVs) {parameterName=Texture, samplerSource=SSM_Clamp_WorldGroupSettings, texture=/FortUILibrary/Foundation/Buttons/Textures/T_SkewButtonShape_B_DF}
  n4 = ScalarParameter() {defaultValue=0.1, parameterName=EdgeFilter}
  n5 = MaterialFunctionCall!(Threshold=@n1, Gradient=@n3.R, Filter Width Override, Filter Width Fudge Factor=@n4) {materialFunction=/Game/UI/Materials/AAThreshold}
  LinearInterpolate(
   B=LinearInterpolate(
    A=VectorParameter() {parameterName=ShapeColor, defaultValue=<rgba 1 1 1 1>}.RGB
    B=VectorParameter() {parameterName=HoverColor, defaultValue=<rgba 1 0.9444 0.4103 1>}.RGB
    Alpha=ScalarParameter() {parameterName=HoverAmount}
   )
   Alpha=@n5.Result
  )
  ScalarParameter() {defaultValue=1, parameterName=OuterOpacity}
  n6 = Add(A=@n1, B=ScalarParameter() {defaultValue=0.05, parameterName=OuterEdgeSize})
  n7 = TextureSampleParameter2D(UVs=@n2.UVs) {parameterName=Texture, samplerSource=SSM_Clamp_WorldGroupSettings, texture=&T_SkewButtonShape_B_DF}
  Subtract(
   A=@n5.Result
   B=Multiply(
    A=MaterialFunctionCall(Threshold=@n6, Gradient=@n7.R, Filter Width Fudge Factor=@n4) {materialFunction=&AAThreshold}.Result
    B=ScalarParameter() {defaultValue=0.1, parameterName=InnerOpacity}
   )
  )
 ]
}
