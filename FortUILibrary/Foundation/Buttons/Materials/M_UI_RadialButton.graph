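# M_UI_RadialButton is a translucent UI material that draws a procedural circular button as an inner disc with an optional outer ring.
# Normalized UI UVs are remapped around the center and fed into two SphereMasks, each with radius and hardness parameters.
# The inverted inner mask cuts the ring's color, while outer color alpha scales the outer mask in the combined opacity.
# The UseSingleCircle static switch reduces both color and mask to the inner disc alone.
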Material /FortUILibrary/Foundation/Buttons/Materials/M_UI_RadialButton
graph {
 properties {
  bUsedWithStaticMesh=false
  blendMode=BLEND_Translucent
  materialDomain=MD_UI
  pixelDepthOffsetMode=PDOM_Legacy
  usedShadingModels=Default Lit
 }
 nodes [
  n1 = Constant2Vector()
  n2 = MaterialFunctionCall() {materialFunction=&GetUserInterfaceUV}
  n3 = Multiply(A=Subtract(A=ComponentMask(None=@n2.Normalized UV) {g=true, r=true}) {constB=0.5}) {constB=2}
  n4 = SphereMask(
   A=@n3
   B=@n1
   Radius=ScalarParameter() {defaultValue=0.8, parameterName=InnerCircleRadius}
   Hardness=ScalarParameter() {defaultValue=0.97, parameterName=InnerCircleHardness}
  )
  n5 = Multiply(A=VectorParameter() {parameterName=InnerCircleColor, defaultValue=<rgba 0 0.1859 1 1>}.RGB, B=@n4)
  n6 = SphereMask(
   A=@n3
   B=@n1
   Radius=ScalarParameter() {defaultValue=1, parameterName=OuterCircleRadius}
   Hardness=ScalarParameter() {defaultValue=0.97, parameterName=OuterCircleHardness}
  )
  n7 = Multiply(A=OneMinus(None=@n4), B=@n6)
  n8 = VectorParameter() {parameterName=OuterCircleColor, defaultValue=<rgba 1 0.8508 0 1>}
  StaticSwitchParameter(True=@n5, False=Add(A=@n5, B=Multiply(A=@n7, B=@n8.RGB))) {parameterName=UseSingleCircle}
  n9 = Multiply(A=@n6, B=@n8.A)
  StaticSwitchParameter(True=@n4, False=Clamp(None=Add(A=@n4, B=@n9))) {parameterName=UseSingleCircle}
 ]
}
