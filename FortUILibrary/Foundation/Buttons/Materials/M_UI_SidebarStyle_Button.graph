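# UI material for sidebar-style buttons, with animated selection, highlight and warning states.
# The base blends solid or selected color with a curve-atlas gradient; the highlight layer combines a panning, screen-aligned triangle grid, white burst gradient and optional angled left bar.
# With Transition on, AnimateHighlight fades in that layer over its first half, then slides the bar over its second.
# Static switches control left bar, transition and gradient-only styles.
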Material /FortUILibrary/Foundation/Buttons/Materials/M_UI_SidebarStyle_Button
graph {
 properties {
  bUsedWithStaticMesh=false
  blendMode=BLEND_Translucent
  materialDomain=MD_UI
  pixelDepthOffsetMode=PDOM_Legacy
  usedShadingModels=Default Lit
 }
 nodes [
  n1 = VectorParameter() {group=Gradient, parameterName=RGB_SolidFill, defaultValue=<rgba 0 0.0802 0.2664 1>}
  n2 = VectorParameter() {group=Selected, parameterName=RGB_Selected, defaultValue=<rgba 0.0521 0.1779 1 1>}
  n3 = ScalarParameter() {group=Animation, parameterName=AnimateSelection}
  n4 = CurveAtlasRowParameter(CurveTime=ComponentMask(None=MaterialFunctionCall() {materialFunction=&GetUserInterfaceUV}.Normalized UV) {r=true}) {
   atlas=/Game/UI/Sidebar/Materials/CA_Sidebar
   curve=/Game/UI/Sidebar/Materials/CC_Sidebar_UserListBaclground
   group=Gradient
   parameterName=GradientCurve
  }
  n5 = ScalarParameter() {defaultValue=1, group=Gradient, parameterName=GradientFill}
  n6 = LinearInterpolate(A=LinearInterpolate(A=@n1.RGB, B=@n2.RGB, Alpha=@n3), B=@n4, Alpha=@n5)
  n7 = ScalarParameter() {defaultValue=-0.25, group=Left Bar, parameterName=LeftBar_Position}
  n8 = Multiply(A=Abs(None=@n7)) {constB=0.5}
  n9 = ScalarParameter() {group=Animation, parameterName=AnimateHighlight, sliderMax=1}
  n10 = Multiply(A=@n9) {constB=2}
  n11 = LinearInterpolate(B=Multiply(A=Subtract(A=@n9) {constB=0.5}) {constB=2}, Alpha=Saturate(None=Floor(None=@n10))) {desc="AnimateHIghlight <= 0.5 result is 0"}
  n12 = Add(
   A=MaterialFunctionCall() {materialFunction=&GetUserInterfaceUV}.9-Slice UV
   B=StaticSwitchParameter(
    True=LinearInterpolate(A=@n8, B=@n7, Alpha=@n11) {constA=-0.1, constAlpha=1}
    False=ScalarParameter() {defaultValue=-0.25, group=Left Bar, parameterName=LeftBar_Position}
   ) {defaultValue=true, group=Style, parameterName=Transition, sortPriority=1}
  )
  n13 = ScalarParameter() {defaultValue=-0.25, group=Left Bar, parameterName=LeftBar_Angle}
  n14 = Multiply(
   A=ComponentMask(None=Rotator(Coordinate=@n12, Time=@n13)) {r=true}
   B=ScalarParameter() {defaultValue=256, group=Left Bar, parameterName=LeftBar_Sharpness}
  )
  n15 = VectorParameter() {group=Triangles, parameterName=RGB2, defaultValue=<rgba 0.0331 0.402 1 1>}
  n16 = VectorParameter() {parameterName=RGB2_Warning, defaultValue=<rgba 1 1 1 0>}
  n17 = ScalarParameter() {parameterName=AnimateWarning}
  n18 = MaterialFunctionCall(Float2=MaterialFunctionCall() {materialFunction=&ScreenResolution}.Visible Resolution) {materialFunction=&BreakOutFloat2Components}
  n19 = Divide(A=@n18.R, B=Max(A=@n18.G))
  n20 = Constant() {r=1}
  n21 = Rotator(
   Coordinate=Multiply(
    A=AppendVector(A=@n19, B=@n20)
    B=MaterialFunctionCall() {materialFunction=&ScreenAlignedUVs}."X 100%, Y 100%"
   )
   Time=ScalarParameter() {defaultValue=-0.0571, group=Triangles, parameterName=Tri Angle}
  ) {speed=6.2832}
  n22 = AppendVector(A=ScalarParameter() {defaultValue=0.0025, group=Triangles, parameterName=PanSpeed}, B=Constant())
  n23 = TextureSample(UVs=Multiply(A=Panner(Coordinate=@n21, Speed=@n22), B=ScalarParameter() {defaultValue=5, group=Triangles, parameterName=Tri Tile Amount})) {texture=/Game/Global/Textures/Default/Masks_and_Gradients/T-TriangleGrid}
  n24 = ScalarParameter() {defaultValue=0.2, group=Triangles, parameterName=TriangleDensity}
  n25 = VectorParameter() {group=Triangles, parameterName=RGB1, defaultValue=<rgba 0.6038 0.8879 1 1>}
  n26 = VectorParameter() {parameterName=RGB1_Warning, defaultValue=<rgba 1 0.3949 0.5055 1>}
  n27 = ScalarParameter() {group=white gradient, parameterName=white_Burst_start}
  n28 = ScalarParameter() {defaultValue=1, group=white gradient, parameterName=white_Burst_end}
  n29 = ScalarParameter() {group=white gradient, parameterName=Burst}
  n30 = ComponentMask(None=Subtract(
    A=MaterialFunctionCall() {materialFunction=&GetUserInterfaceUV}.Normalized UV
    B=LinearInterpolate(A=@n27, B=@n28, Alpha=@n29)
   )) {r=true}
  n31 = ScalarParameter() {defaultValue=1.5, group=white gradient, parameterName=White_sharpness}
  n32 = LinearInterpolate(
   A=Constant() {r=1}
   B=LinearInterpolate(
    A=LinearInterpolate(A=@n25.RGB, B=@n26.RGB, Alpha=@n17)
    B=LinearInterpolate(A=@n15.RGB, B=@n16.RGB, Alpha=@n17)
    Alpha=Multiply(A=@n23.R, B=@n24)
   )
   Alpha=Saturate(None=Multiply(A=@n30, B=@n31))
  )
  n33 = StaticSwitchParameter(
   True=LinearInterpolate(
    A=VectorParameter() {group=Left Bar, parameterName=LeftBar_Color, defaultValue=<rgba 0.9301 1 0.0103 1>}.RGB
    B=@n32
    Alpha=Saturate(None=@n14)
   )
   False=@n32
  ) {defaultValue=true, group=Style, parameterName=ShowLeftBar, sortPriority=0}
  n34 = Multiply(A=ScalarParameter() {group=Animation, parameterName=AnimateHighlight, sliderMax=1}) {constB=2}
  n35 = LinearInterpolate(A=Saturate(None=@n34), Alpha=Floor(None=@n34)) {desc="AnimateHIghlight >= 0.5 result is 1"}
  StaticSwitchParameter(
   True=@n6
   False=StaticSwitchParameter(True=LinearInterpolate(A=@n6, B=@n33, Alpha=@n35), False=@n33) {defaultValue=true, group=Style, parameterName=Transition, sortPriority=1}
  ) {group=Style, parameterName=GradientOnly, sortPriority=0}
  n36 = LinearInterpolate(A=LinearInterpolate(A=@n1.A, B=@n2.A, Alpha=@n3), B=@n4.A, Alpha=@n5)
  StaticSwitchParameter(True=@n36, False=LinearInterpolate(A=@n36, B=ScalarParameter() {defaultValue=1, group=Style, parameterName=PrimaryAlpha}, Alpha=@n35)) {group=Style, parameterName=GradientOnly, sortPriority=0}
 ]
}
