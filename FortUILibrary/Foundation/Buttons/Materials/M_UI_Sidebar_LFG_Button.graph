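# Translucent UI material for the sidebar LFG button, with an angled, colored left bar beside a triangle-patterned fill.
# A rotated, panning, screen-aligned triangle grid blends hovered and unhovered color pairs, with edge trim derived from two offset texture samples.
# AnimateHighlight feeds a first-half highlight ramp and a second-half bar slide, which ShowTrimAlways overrides.
# Named reroutes carry the ramp, a shine swipe and a white burst gradient.
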Material /FortUILibrary/Foundation/Buttons/Materials/M_UI_Sidebar_LFG_Button
graph {
 properties {
  bUsedWithStaticMesh=false
  blendMode=BLEND_Translucent
  materialDomain=MD_UI
  pixelDepthOffsetMode=PDOM_Legacy
  usedShadingModels=Default Lit
 }
 nodes [
  n1 = Multiply(A=NamedRerouteUsage()) {constB=0.5, desc=Overall shine strength}
  n2 = LinearInterpolate(
   A=VectorParameter() {group=Gradient, parameterName=RGB_SolidFill, defaultValue=<rgba 0.0024 0.063 0.2874 1>}.RGB
   B=Constant() {r=1}
   Alpha=NamedRerouteUsage()
  )
  n3 = MaterialFunctionCall(Float2=MaterialFunctionCall() {materialFunction=&ScreenResolution}.Visible Resolution) {materialFunction=&BreakOutFloat2Components}
  n4 = Max(A=@n3.G)
  n5 = Multiply(
   A=AppendVector(A=Divide(A=@n3.R, B=@n4), B=Constant() {r=1})
   B=MaterialFunctionCall() {materialFunction=&ScreenAlignedUVs}."X 100%, Y 100%"
  )
  n6 = ScalarParameter() {defaultValue=-0.0571, group=Triangles, parameterName=Tri Angle}
  n7 = ScalarParameter() {defaultValue=0.0025, group=Triangles, parameterName=PanSpeed}
  n8 = Constant()
  n9 = Multiply(
   A=Panner(Coordinate=Rotator(Coordinate=@n5, Time=@n6) {speed=6.2832}, Speed=AppendVector(A=@n7, B=@n8))
   B=ScalarParameter() {defaultValue=5, group=Triangles, parameterName=Tri Tile Amount}
  )
  n10 = MaterialFunctionCall(Float2=@n9) {materialFunction=&BreakOutFloat2Components}
  n11 = TextureSample(UVs=AppendVector(A=Add(A=@n10.R) {constB=0.01}, B=Add(A=@n10.G) {constB=-0.01})) {texture=/Game/Global/Textures/Default/Masks_and_Gradients/T-TriangleGrid}
  n12 = Multiply(A=@n11.R, B=ScalarParameter() {defaultValue=0.2, group=Triangles, parameterName=TriangleDensity})
  n13 = LinearInterpolate(
   A=VectorParameter() {group=Triangles, parameterName=UnHovered_RGB1, defaultValue=<rgba 0.6038 0.8879 1 1>}.RGB
   B=VectorParameter() {group=Triangles, parameterName=UnHovered_RGB2, defaultValue=<rgba 0.0331 0.402 1 1>}.RGB
   Alpha=@n12
  )
  n14 = Multiply(A=Subtract(A=@n11.R, B=TextureSample(UVs=@n9) {texture=&T-TriangleGrid}.R)) {constB=100}
  n15 = Add(
   A=@n13
   B=Multiply(A=Saturate(None=@n14), B=ScalarParameter() {defaultValue=0.03, group=Shineable, parameterName=TriangleTrimStrength}) {constB=0.03, desc=controls the trim strength}
  ) {desc=adds shine at triangle borders}
  n16 = ScalarParameter() {parameterName=ShowTrimAlways}
  n17 = LinearInterpolate(
   A=LinearInterpolate(A=@n13, B=LinearInterpolate(A=@n13, B=@n15, Alpha=@n16), Alpha=NamedRerouteUsage())
   B=LinearInterpolate(
    A=VectorParameter() {group=Triangles, parameterName=Hovered_RGB1, defaultValue=<rgba 0.6038 0.8879 1 1>}.RGB
    B=VectorParameter() {group=Triangles, parameterName=Hovered_RGB2, defaultValue=<rgba 0.0331 0.402 1 1>}.RGB
    Alpha=@n12
   )
   Alpha=NamedRerouteUsage()
  )
  n18 = NamedRerouteUsage()
  n19 = MaterialFunctionCall() {materialFunction=&GetUserInterfaceUV}
  n20 = ScalarParameter() {defaultValue=-0.25, group=Left Bar, parameterName=LeftBar_Position}
  n21 = Abs(None=@n20)
  n22 = ScalarParameter() {group=Animation, parameterName=AnimateHighlight, sliderMax=1}
  n23 = Multiply(A=Subtract(A=@n22) {constB=0.5}) {constB=2}
  n24 = Saturate(None=Floor(None=Multiply(A=@n22) {constB=2}))
  n25 = LinearInterpolate(
   A=LinearInterpolate(A=Multiply(A=@n21) {constB=0.5}, B=@n20, Alpha=LinearInterpolate(B=@n23, Alpha=@n24) {desc="AnimateHIghlight <= 0.5 result is 0"}) {constA=-0.1, constAlpha=1}
   B=ScalarParameter() {defaultValue=-0.25, group=Left Bar, parameterName=LeftBar_Position}
   Alpha=ScalarParameter() {parameterName=ShowTrimAlways}
  )
  n26 = ComponentMask(None=Rotator(Coordinate=Add(A=@n19.9-Slice UV, B=@n25), Time=ScalarParameter() {defaultValue=-0.25, group=Left Bar, parameterName=LeftBar_Angle})) {r=true}
  n27 = ScalarParameter() {defaultValue=256, group=Left Bar, parameterName=LeftBar_Sharpness}
  LinearInterpolate(
   A=VectorParameter() {group=Left Bar, parameterName=LeftBar_UnHovered_Color, defaultValue=<rgba 0 0.6376 1 1>}.RGB
   B=Multiply(A=Add(A=@n1), B=LinearInterpolate(A=@n2, B=@n17, Alpha=@n18))
   Alpha=Saturate(None=Multiply(A=@n26, B=@n27))
  )
  n28 = Constant()
  n29 = MaterialFunctionCall(Float2=MaterialFunctionCall() {materialFunction=&GetUserInterfaceUV}.Normalized UV) {materialFunction=&BreakOutFloat2Components}
  n30 = MaterialFunctionCall!(In_Uv=AppendVector(A=Multiply(A=@n29.R) {constB=0.33}, B=@n29.G)) {materialFunction=/Game/UI/Foundation/Materials/M-Function-ShineSwipe}
  n31 = ScalarParameter() {parameterName=Shine}
  NamedRerouteDeclaration(None=LinearInterpolate(A=@n28, B=@n30.Result, Alpha=@n31)) {name=ShineResult, nodeColor=<rgba 0.3594 0 1 1>}
  n32 = ScalarParameter() {group=white gradient, parameterName=white_Burst_start}
  n33 = ScalarParameter() {defaultValue=1, group=white gradient, parameterName=white_Burst_end}
  n34 = ScalarParameter() {group=white gradient, parameterName=Burst}
  n35 = ComponentMask(None=Subtract(
    A=MaterialFunctionCall() {materialFunction=&GetUserInterfaceUV}.Normalized UV
    B=LinearInterpolate(A=@n32, B=@n33, Alpha=@n34)
   )) {r=true}
  n36 = ScalarParameter() {defaultValue=1.5, group=white gradient, parameterName=White_sharpness}
  NamedRerouteDeclaration(None=Saturate(None=Multiply(A=@n35, B=@n36))) {name=Whiteburst, nodeColor=<rgba 0.9453 0 1 1>}
  n37 = Multiply(A=ScalarParameter() {group=Animation, parameterName=AnimateHighlight, sliderMax=1}) {constB=2}
  NamedRerouteDeclaration(None=LinearInterpolate(A=Saturate(None=@n37), Alpha=Floor(None=@n37)) {desc="AnimateHIghlight >= 0.5 result is 1"}) {name=AnimHighlightLerp, nodeColor=<rgba 1 0.9141 0 1>}
  NamedRerouteDeclaration(None=LinearInterpolate(A=@n28, B=@n30.Alpha, Alpha=@n31)) {name=ShineAlpha, nodeColor=<rgba 0 0.6953 1 1>}
 ]
}
