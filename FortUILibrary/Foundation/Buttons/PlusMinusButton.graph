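# CommonButtonBase widget for a plus/minus style button, with a configurable text symbol (default "-") over a solid background and an input action prompt.
# PreConstruct applies the text and font size, then aligns the prompt from IsBottom (top-center when set) or IsLeft (left or right).
# HoverAnim plays forward or in reverse on hover changes, inverting the navy/white colors and scaling up the text.
# Enable and disable events recolor both layers and reset text scale.
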WidgetBlueprint /FortUILibrary/Foundation/Buttons/PlusMinusButton
animations [HoverAnim {
  binding_count=2
  length_seconds=0.1
  object=&self:HoverAnim
  sequence {bindings [
    ButtonText {
     id=#1
     possessed_class=CommonTextBlock
     tracks [
      {
       type=MovieSceneColorTrack
       display_name=Color and Opacity
       sections [{
         type=MovieSceneColorSection
         channels [
          Color.R {type=MovieSceneScriptingFloatChannel, keys [[0,1],[2,0]]}
          Color.G {type=MovieSceneScriptingFloatChannel, keys [[0,1],[2,0.005]]}
          Color.B {type=MovieSceneScriptingFloatChannel, keys [[0,1],[2,0.13]]}
          Color.A {type=MovieSceneScriptingFloatChannel}
         ]
        }]
      }
      {
       type=MovieScene2DTransformTrack
       sections [{
         type=MovieScene2DTransformSection
         channels [
          Translation.X {type=MovieSceneScriptingFloatChannel}
          Translation.Y {type=MovieSceneScriptingFloatChannel}
          Angle {type=MovieSceneScriptingFloatChannel}
          Scale.X {type=MovieSceneScriptingFloatChannel, keys [[0,0.8],[2,1]]}
          Scale.Y {type=MovieSceneScriptingFloatChannel, keys [[0,0.8],[2,1]]}
          Shear.X {type=MovieSceneScriptingFloatChannel}
          Shear.Y {type=MovieSceneScriptingFloatChannel}
         ]
        }]
      }
     ]
    }
    BackgroundColor {
     id=#2
     possessed_class=Image
     tracks [{
       type=MovieSceneColorTrack
       display_name=Color and Opacity
       sections [{
         type=MovieSceneColorSection
         channels [
          Color.R {type=MovieSceneScriptingFloatChannel, keys [[0,0],[2,1]]}
          Color.G {type=MovieSceneScriptingFloatChannel, keys [[0,0.0056],[2,1]]}
          Color.B {type=MovieSceneScriptingFloatChannel, keys [[0,0.1329],[2,1]]}
          Color.A {type=MovieSceneScriptingFloatChannel}
         ]
        }]
      }]
    }
   ]}
 }]
graphs [EventGraph {nodes [
   Event PreConstruct()
   > "SetText (Text)"(self=Get ButtonText() -> (ButtonText:Common Text Object Reference), InText=Get In Text() -> (In Text:Text))
   n1 = Get ButtonText()
   n2 = Break Slate Font Info!(SlateFontInfo:"Slate Font Info Structure (by ref)"=Get Font!(self:Text Object Reference=@n1) -> (Font:Slate Font Info Structure)) -> (
    FontObject:Object Reference
    FontMaterial:Object Reference
    OutlineSettings:Font Outline Settings Structure
    TypefaceFontName:Name
    Size:float=~
    LetterSpacing:int
    SkewAmount:float=~
    bForceMonospaced:bool=~
    bMaterialIsStencil:bool=~
    MonospacedWidth:float=~
   ) {is=title K2Node_BreakStruct}
   n3 = Get Font Size() -> (Font Size:int=0)
   > SetFont(
    self=@n1
    InFontInfo=MakeSlateFontInfo!(
     FontObject:Object Reference=@n2
     FontMaterial:Object Reference=@n2.FontMaterial
     OutlineSettings:Font Outline Settings Structure=@n2.OutlineSettings
     TypefaceFontName:Name=@n2.TypefaceFontName
     Size:float="To Float (Integer)"(InInt=@n3)
     LetterSpacing:int=@n2.LetterSpacing
     SkewAmount:float=0
     bForceMonospaced:bool=false
     bMaterialIsStencil:bool=false
     MonospacedWidth:float=1
    ) -> (ReturnValue:Slate Font Info Structure)
   )
   n4 = SlotAsOverlaySlot(Widget=Get Overlay_ButtonPrompt() -> (Overlay_ButtonPrompt:Overlay Object Reference))
   n5 = Get IsLeft() -> (IsLeft:bool)
   n6 = Get IsBottom() -> (IsBottom:bool)
   > SetHorizontalAlignment(
    self=@n4
    InHorizontalAlignment=Select(
     Option 0:EHorizontalAlignment Enum=Select(Option 0=HAlign_Right, Option 1=HAlign_Left, Index=@n5)
     Option 1:EHorizontalAlignment Enum=HAlign_Center
     Index=@n6
    ) -> (ReturnValue:EHorizontalAlignment Enum)
   )
   > SetVerticalAlignment(
    self=@n4
    InVerticalAlignment=Select(Option 0:EVerticalAlignment Enum=VAlign_Center, Option 1:EVerticalAlignment Enum=VAlign_Top, Index=@n6) -> (ReturnValue:EVerticalAlignment Enum)
   )
   Event On Enabled() {is=title K2Node_Event}
   > SetColorAndOpacity(
    self:Image Object Reference=Get BackgroundColor() -> (BackgroundColor:Image Object Reference)
    InColorAndOpacity:Linear Color Structure=(R=0,G=0.005,B=0.13,A=1)
   )
   > n7 = SetColorAndOpacity!(
    execute
    self:Text Object Reference=Get ButtonText()
    InColorAndOpacity_SpecifiedColor:Linear Color Structure=(R=1,G=1,B=1,A=1)
    InColorAndOpacity_ColorUseRule:ESlateColorStylingMode Enum
   ) -> (then)
   SetRenderScale!(execute=[@n8.then, @n7.then], self:Widget Object Reference=Get ButtonText(), Scale:Vector 2D Structure=(X=0.8,Y=0.8)) -> (then)
   Event On Disabled() {is=title K2Node_Event}
   > SetColorAndOpacity!(execute, self:Image Object Reference=Get BackgroundColor(), InColorAndOpacity:Linear Color Structure=(R=0.2,G=0.2,B=0.2,A=0.5)) -> (then)
   > n8 = SetColorAndOpacity!(
    execute
    self:Text Object Reference=Get ButtonText()
    InColorAndOpacity_SpecifiedColor:Linear Color Structure=(R=1,G=1,B=1,A=0.5)
    InColorAndOpacity_ColorUseRule:ESlateColorStylingMode Enum
   ) -> (then)
   Event On Hovered()
   n9 = Get HoverAnim() -> (HoverAnim:Widget Animation Object Reference)
   > PlayAnimationForward(InAnimation=@n9)
   Event On Unhovered()
   > PlayAnimationReverse(InAnimation=@n9)
  ]}]
variables [
 AllowDisabledAnimation {type=boolean}
 Font Size {type=integer, default=60}
 In Text {type=string, default=-}
 IsBottom {type=boolean, default=true}
 IsLeft {type=boolean, default=true}
]
widgets {
 info {inheritedWidgetCount=1, parentClass=&CommonButtonBase, rootWidgetClass=&Overlay, widgetCount=5}
 widgets [
  Overlay_10 {widgetClassPath=&Overlay, widget=Overlay_10}
  BackgroundColor {
   widgetClassPath=&Image
   bIsVariable=true
   parent=Overlay_10
   widget=BackgroundColor
   properties {colorAndOpacity=<rgba 0 0.005 0.13 1>}
   slot {class=OverlaySlot, properties {horizontalAlignment=HAlign_Fill, verticalAlignment=VAlign_Fill}}
  }
  Overlay_ButtonPrompt {
   widgetClassPath=&Overlay
   bIsVariable=true
   parent=Overlay_10
   widget=Overlay_ButtonPrompt
   slot {class=OverlaySlot, properties {horizontalAlignment=HAlign_Center, verticalAlignment=VAlign_Center, padding=<ltrb -50 -16 -50 -16>}}
  }
  InputActionWidget {
   widgetClassPath=&CommonActionWidget
   bInherited=true
   bIsVariable=true
   parent=Overlay_ButtonPrompt
   widget=InputActionWidget
   properties {renderTransform {angle=0, scale=<1.25 1.25>, shear=<0 0>, translation=<0 0>}}
   slot {class=OverlaySlot}
  }
  ButtonText {
   widgetClassPath=&CommonTextBlock
   bIsVariable=true
   parent=Overlay_10
   widget=ButtonText
   properties {
    text=-
    font {
     bForceMonospaced=false
     bMaterialIsStencil=false
     fontMaterial=null
     fontObject=&Burbank
     letterSpacing=0
     monospacedWidth=1
     size=60
     skewAmount=0
     typefaceFontName=Burbank-PlayerPower
     outlineSettings {
      bApplyOutlineToDropShadows=false
      bMiteredCorners=false
      bSeparateFillAlpha=false
      outlineMaterial=null
      outlineSize=0
      outlineColor=<rgba 0 0 0 1>
     }
    }
    margin=<ltrb 0 -32 0 -32>
    renderTransform {angle=0, scale=<0.8 0.8>, shear=<-8 0>, translation=<0 0>}
    shadowOffset=<0 0>
   }
   slot {class=OverlaySlot, properties {horizontalAlignment=HAlign_Center, verticalAlignment=VAlign_Center}}
  }
 ]
}
