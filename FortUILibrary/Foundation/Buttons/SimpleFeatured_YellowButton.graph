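# Fortnite UI widget for a yellow featured button, built on FortSimpleFeaturedYellowButton, with an input action icon, main label and optional secondary text.
# RefreshUI, called on PreConstruct and Construct, applies font size and padding overrides, label text, shear, and a backing material with or without shine swipe.
# Hover plays text color, material parameter and text bounce animations, reversing color and material on unhover; clicks ping-pong the Pressed animation.
# Secondary text collapses when empty.
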WidgetBlueprint /FortUILibrary/Foundation/Buttons/SimpleFeatured_YellowButton
animations [
 Hover {
  binding_count=2
  length_seconds=0.2
  object=&self:Hover
  sequence {bindings [
    Border_FontColorChange {
     id=#1
     possessed_class=Border
     tracks [{
       type=MovieSceneColorTrack
       display_name=Content Color and Opacity
       sections [{
         type=MovieSceneColorSection
         end_frame=4
         start_frame=0
         channels [
          Color.R {type=MovieSceneScriptingFloatChannel, const [0,1]}
          Color.G {type=MovieSceneScriptingFloatChannel, keys [[0,0.004],[4,0.01]]}
          Color.B {type=MovieSceneScriptingFloatChannel, keys [[0,0.0242],[4,0.1]]}
          Color.A {type=MovieSceneScriptingFloatChannel, const [1,1]}
         ]
        }]
      }]
    }
    ButtonBacking {
     id=#2
     possessed_class=Image
     tracks [{
       type=MovieSceneWidgetMaterialTrack
       sections [{
         type=MovieSceneParameterSection
         end_frame=4
         start_frame=0
         channels [
          Animate_BackingColor {type=MovieSceneScriptingFloatChannel, keys [[0,0],[4,1]]}
          Animate_BorderColor_Hover {type=MovieSceneScriptingFloatChannel, keys [[0,0],[4,1]]}
         ]
        }]
      }]
    }
   ]}
 }
 TextBounce {
  binding_count=1
  length_seconds=0.2
  object=&self:TextBounce
  sequence {bindings [Text_ButtonAction {
     id=#3
     possessed_class=CommonTextBlock
     tracks [{
       type=MovieScene2DTransformTrack
       sections [{
         type=MovieScene2DTransformSection
         end_frame=4
         start_frame=0
         channels [
          Translation.X {type=MovieSceneScriptingFloatChannel}
          Translation.Y {type=MovieSceneScriptingFloatChannel, keys [[0,0],[1,-4],[4,0]]}
          Angle {type=MovieSceneScriptingFloatChannel}
          Scale.X {type=MovieSceneScriptingFloatChannel}
          Scale.Y {type=MovieSceneScriptingFloatChannel}
          Shear.X {type=MovieSceneScriptingFloatChannel}
          Shear.Y {type=MovieSceneScriptingFloatChannel}
         ]
        }]
      }]
    }]}
 }
 Pressed {
  binding_count=1
  length_seconds=0.1
  object=&self:Pressed
  sequence {bindings [ButtonBacking {
     id=#4
     possessed_class=Image
     tracks [{
       type=MovieSceneWidgetMaterialTrack
       sections [{
         type=MovieSceneParameterSection
         end_frame=2
         start_frame=0
         channels [Animate_BackingColor_Selected {type=MovieSceneScriptingFloatChannel, keys [[0,0],[2,1]]}]
        }]
      }]
    }]}
 }
]
graphs [
 EventGraph {nodes [
   n1 = Event PreConstruct()
   RefreshUI(execute=[@n1.then, @n2.then], self:Fort Simple Featured Yellow Button Object Reference)
   n2 = Event Construct()
   n3 = Event SetText!() -> (OutputDelegate, then, NewText:"Text (by ref)") {is=title K2Node_Event}
   > n4 = Set Button Description(Button Description:Text=@n3.NewText) -> (Output_Get:Text)
   > "SetText (Text)"(self=Get Text_ButtonAction() -> (Text_ButtonAction:Common Text Object Reference), InText=@n4)
   Event On Clicked()
   > PlayAnimation(InAnimation=Get Pressed() -> (Pressed:Widget Animation Object Reference), PlayMode=PingPong)
   n5 = Event ChangeSecondaryText!() -> (OutputDelegate, then, NewText:"Text (by ref)") {is=title K2Node_Event}
   n6 = Get Text_SecondaryText() -> (Text_SecondaryText:Common Text Object Reference)
   > SetVisibility(self=@n6, InVisibility=Select(Option 0=SelfHitTestInvisible, Option 1=Collapsed, Index=TextIsEmpty(InText=@n5.NewText)))
   > "SetText (Text)"(self=@n6, InText=@n5.NewText)
   Event On Hovered()
   n7 = Get Hover() -> (Hover:Widget Animation Object Reference)
   > PlayAnimationForward(InAnimation=@n7, PlaybackSpeed=1.5)
   > PlayAnimation(InAnimation=Get TextBounce() -> (TextBounce:Widget Animation Object Reference), PlaybackSpeed=1.5)
   Event On Unhovered()
   > PlayAnimationReverse(InAnimation=@n7)
   Event RefreshUI() {is=title K2Node_Event}
   > HandleSize()
   > SetText!(
    execute
    self:Fort Simple Featured Yellow Button Object Reference
    NewText:"Text (by ref)"=Get Button Description() -> (Button Description:Text)
   ) -> (then)
   n8 = Get ButtonBacking() -> (ButtonBacking:Image Object Reference)
   > SetBrushFromMaterial(
    self=@n8
    Material=Select(
     Option 0:Material Interface Object Reference=/FortUILibrary/Foundation/Art_ToolKit/AngledBoxes/M_UI_SimpleFeatured_YellowButton_NoShine
     Option 1:Material Interface Object Reference=/FortUILibrary/Foundation/Art_ToolKit/AngledBoxes/M_UI_SimpleFeatured_YellowButton
     Index=Get bUseShineSwipe() -> (bUseShineSwipe:bool)
    ) -> (ReturnValue:Material Interface Object Reference)
   )
   > n9 = GetDynamicMaterial(self=@n8)
   > SetScalarParameterValue(self=@n9, ParameterName=DelayShineCycle, Value=Get DelayShine() -> (DelayShine:float))
   n10 = Get bShearText() -> (bShearText:bool)
   > SetRenderShear!(
    execute
    self:Widget Object Reference=Get Text_ButtonAction()
    Shear:Vector 2D Structure=MakeVector2D(X=Select(Option 0:double=0, Option 1:double=-12, Index=@n10) -> (ReturnValue:double), Y=0)
   ) -> (then)
  ]}
 HandleSize {nodes [
   HandleSize() {is=title K2Node_FunctionEntry}
   n1 = Select(Option 0:int=22, Option 1:int=Get FontSize() -> (FontSize:int=0), Index=Get bFontSizeOverride() -> (bFontSizeOverride:bool)) -> (ReturnValue:int)
   > SetFont(
    self=Get Text_ButtonAction()
    InFontInfo=MakeSlateFontInfo!(
     FontObject:Object Reference=&Burbank
     FontMaterial:Object Reference
     OutlineSettings:Font Outline Settings Structure=(OutlineSize=0,bMiteredCorners=False,bSeparateFillAlpha=False,bApplyOutlineToDropShadows=False,OutlineMaterial=None,OutlineColor=(R=0,G=0,B=0,A=1))
     TypefaceFontName:Name=BurbankRegular-Black
     Size:float="To Float (Integer)"(InInt=@n1)
     LetterSpacing:int=0
     SkewAmount:float=0
     bForceMonospaced:bool=false
     bMaterialIsStencil:bool=false
     MonospacedWidth:float=1
    ) -> (ReturnValue:Slate Font Info Structure)
   )
   > SetPadding!(
    execute
    self:Border Object Reference=Get Border_Container() -> (Border_Container:Border Object Reference)
    InPadding:Margin Structure=Select(
     Option 0:Margin Structure=Make Margin(Left=48, Top=14, Right=48, Bottom=14)
     Option 1:Margin Structure=Get PaddingOverrideMargin() -> (PaddingOverrideMargin:Margin Structure)
     Index=Get bPaddingOverride() -> (bPaddingOverride:bool)
    ) -> (ReturnValue:Margin Structure)
   ) -> (then)
  ]}
]
variables [Button Description {type=string, default="{TEXT}"}]
widgets {
 info {inheritedWidgetCount=1, parentClass=/Script/FortniteUI.FortSimpleFeaturedYellowButton, rootWidgetClass=&SizeBox, widgetCount=10}
 widgets [
  SizeBox_Control {widgetClassPath=&SizeBox, widget=SizeBox_Control, properties {widthOverride=256}}
  Overlay_29 {widgetClassPath=&Overlay, parent=SizeBox_Control, widget=Overlay_29, slot {class=SizeBoxSlot}}
  ButtonBacking {
   widgetClassPath=&Image
   bIsVariable=true
   parent=Overlay_29
   widget=ButtonBacking
   properties {brush {
     bIsDynamicallyLoaded=false
     drawAs=Image
     imageType=NoImage
     mirroring=NoMirror
     resourceName=None
     resourceObject=&M_UI_SimpleFeatured_YellowButton
     tiling=NoTile
     imageSize=<32 32>
     margin=<ltrb 0 0 0 0>
     outlineSettings {
      bUseBrushTransparency=false
      roundingType=HalfHeightRadius
      width=0
      color {colorUseRule=UseColor_Specified, specifiedColor=<rgba 0 0 0 0>}
      cornerRadii=<0 0 0 0>
     }
     tintColor {colorUseRule=UseColor_Specified, specifiedColor=<rgba 1 1 1 1>}
     uVRegion {bIsValid=false, max=<0 0>, min=<0 0>}
    }}
   slot {class=OverlaySlot, properties {horizontalAlignment=HAlign_Fill, verticalAlignment=VAlign_Fill}}
  }
  Border_Container {
   widgetClassPath=&Border
   bIsVariable=true
   parent=Overlay_29
   widget=Border_Container
   properties {
    horizontalAlignment=HAlign_Center
    verticalAlignment=VAlign_Center
    background {
     bIsDynamicallyLoaded=false
     drawAs=NoDrawType
     imageType=NoImage
     mirroring=NoMirror
     resourceName=None
     resourceObject=null
     tiling=NoTile
     imageSize=<32 32>
     margin=<ltrb 0 0 0 0>
     outlineSettings {
      bUseBrushTransparency=false
      roundingType=HalfHeightRadius
      width=0
      color {colorUseRule=UseColor_Specified, specifiedColor=<rgba 0 0 0 0>}
      cornerRadii=<0 0 0 0>
     }
     tintColor {colorUseRule=UseColor_Specified, specifiedColor=<rgba 1 1 1 1>}
     uVRegion {bIsValid=false, max=<0 0>, min=<0 0>}
    }
    padding=<ltrb 0 0 0 0>
   }
   slot {class=OverlaySlot, properties {horizontalAlignment=HAlign_Fill, verticalAlignment=VAlign_Fill}}
  }
  HorizontalBox_0 {
   widgetClassPath=&HorizontalBox
   parent=Border_Container
   widget=HorizontalBox_0
   slot {class=BorderSlot, properties {horizontalAlignment=HAlign_Center, verticalAlignment=VAlign_Center, padding=<ltrb 0 0 0 0>}}
  }
  InputActionWidget {
   widgetClassPath=&CommonActionWidget
   bInherited=true
   bIsVariable=true
   parent=HorizontalBox_0
   widget=InputActionWidget
   slot {class=HorizontalBoxSlot, properties {horizontalAlignment=HAlign_Center, verticalAlignment=VAlign_Center, padding=<ltrb 0 0 8 0>}}
  }
  Border_FontColorChange {
   widgetClassPath=&Border
   parent=HorizontalBox_0
   widget=Border_FontColorChange
   properties {
    horizontalAlignment=HAlign_Center
    background {
     bIsDynamicallyLoaded=false
     drawAs=NoDrawType
     imageType=NoImage
     mirroring=NoMirror
     resourceName=None
     resourceObject=null
     tiling=NoTile
     imageSize=<32 32>
     margin=<ltrb 0 0 0 0>
     outlineSettings {
      bUseBrushTransparency=false
      roundingType=HalfHeightRadius
      width=0
      color {colorUseRule=UseColor_Specified, specifiedColor=<rgba 0 0 0 0>}
      cornerRadii=<0 0 0 0>
     }
     tintColor {colorUseRule=UseColor_Specified, specifiedColor=<rgba 1 1 1 1>}
     uVRegion {bIsValid=false, max=<0 0>, min=<0 0>}
    }
    contentColorAndOpacity=<rgba 0 0.004 0.0242 1>
    padding=<ltrb 0 0 0 0>
   }
   slot {class=HorizontalBoxSlot, properties {horizontalAlignment=HAlign_Center, verticalAlignment=VAlign_Center, size {sizeRule=Fill, value=1}}}
  }
  VerticalBox_0 {
   widgetClassPath=&VerticalBox
   parent=Border_FontColorChange
   widget=VerticalBox_0
   slot {class=BorderSlot, properties {horizontalAlignment=HAlign_Center, padding=<ltrb 0 0 0 0>}}
  }
  Text_ButtonAction {
   widgetClassPath=&CommonTextBlock
   bIsVariable=true
   parent=VerticalBox_0
   widget=Text_ButtonAction
   properties {
    justification=Center
    scrollStyle=&TextScrollStyle-Base_C
    text="{TEXT} <stronger>Hi</>"
    textTransformPolicy=ToUpper
    font {
     bForceMonospaced=false
     bMaterialIsStencil=false
     fontMaterial=null
     fontObject=&Burbank
     letterSpacing=0
     monospacedWidth=1
     size=22
     skewAmount=0
     typefaceFontName=BurbankRegular-Black
     outlineSettings {
      bApplyOutlineToDropShadows=false
      bMiteredCorners=false
      bSeparateFillAlpha=false
      outlineMaterial=null
      outlineSize=0
      outlineColor=<rgba 0 0 0 1>
     }
    }
    shadowOffset=<0 0>
   }
   slot {class=VerticalBoxSlot}
  }
  Text_SecondaryText {
   widgetClassPath=&CommonTextBlock
   bIsVariable=true
   parent=VerticalBox_0
   widget=Text_SecondaryText
   properties {
    justification=Center
    text=$9.49
    visibility=Collapsed
    font {
     bForceMonospaced=false
     bMaterialIsStencil=false
     fontMaterial=null
     fontObject=&BurbankSmall
     letterSpacing=0
     monospacedWidth=1
     size=14
     skewAmount=0
     typefaceFontName=Black
     outlineSettings {
      bApplyOutlineToDropShadows=false
      bMiteredCorners=false
      bSeparateFillAlpha=false
      outlineMaterial=null
      outlineSize=0
      outlineColor=<rgba 0 0 0 1>
     }
    }
    renderTransform {angle=0, scale=<1 1>, shear=<-8 0>, translation=<0 0>}
    shadowOffset=<0 0>
   }
   slot {class=VerticalBoxSlot, properties {horizontalAlignment=HAlign_Center, padding=<ltrb 0 -4 0 0>}}
  }
 ]
}
