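# Small back button widget built on CommonButtonBase, showing a left arrow image beside a CommonActionWidget for the Back input action.
# On initialization, touch input swaps the row's padding for larger margins, and the arrow is a FortMobileImage with a mobile size multiplier.
# Hover plays a short animation that nudges the arrow left and enlarges it, reversed on unhover.
# The On Clicked event is empty in this graph.
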WidgetBlueprint /FortUILibrary/Foundation/Buttons/BackButtonSmall
animations [OnHover {
  binding_count=1
  length_seconds=0.15
  object=&self:OnHover
  sequence {bindings [Image_Arrow {
     id=#1
     possessed_class=FortMobileImage
     tracks [{
       type=MovieScene2DTransformTrack
       sections [{
         type=MovieScene2DTransformSection
         end_frame=3
         start_frame=0
         channels [
          Translation.X {type=MovieSceneScriptingFloatChannel, keys [[0,0],[3,-6]]}
          Translation.Y {type=MovieSceneScriptingFloatChannel, const [0,1]}
          Angle {type=MovieSceneScriptingFloatChannel}
          Scale.X {type=MovieSceneScriptingFloatChannel, keys [[0,1],[3,1.1]]}
          Scale.Y {type=MovieSceneScriptingFloatChannel, keys [[0,1],[3,1.1]]}
          Shear.X {type=MovieSceneScriptingFloatChannel}
          Shear.Y {type=MovieSceneScriptingFloatChannel}
         ]
        }]
      }]
    }]}
 }]
graphs [EventGraph {nodes [
   Event OnInitialized()
   > Branch(Condition=IsUsingTouch(PlayerController=GetOwningPlayer()))
   > SetPadding!(
    execute
    self:Overlay Slot Object Reference=SlotAsOverlaySlot(Widget=Get HorizontalBox_3() -> (HorizontalBox_3:Horizontal Box Object Reference))
    InPadding:Margin Structure=Make Margin(Left=40, Top=30, Right=40, Bottom=40)
   ) -> (then)
   Event On Hovered()
   > PlayAnimation(InAnimation=Get OnHover() -> (OnHover:Widget Animation Object Reference))
   Event On Unhovered()
   > PlayAnimation(InAnimation=Get OnHover(), PlayMode=Reverse)
   Event On Clicked()
  ]}]
variables []
widgets {
 info {inheritedWidgetCount=1, parentClass=&CommonButtonBase, rootWidgetClass=&Overlay, widgetCount=5}
 widgets [
  Overlay_141 {widgetClassPath=&Overlay, widget=Overlay_141}
  HorizontalBox_3 {
   widgetClassPath=&HorizontalBox
   bIsVariable=true
   parent=Overlay_141
   widget=HorizontalBox_3
   slot {class=OverlaySlot, properties {padding=<ltrb 20 20 20 20>}}
  }
  Image_Arrow {
   widgetClassPath=/Script/FortniteUI.FortMobileImage
   bIsVariable=true
   parent=HorizontalBox_3
   widget=Image_Arrow
   properties {
    mobileSizeMultiplier=2
    brush {
     bIsDynamicallyLoaded=false
     drawAs=Image
     imageType=NoImage
     mirroring=NoMirror
     resourceName=None
     resourceObject=/FortUILibrary/Foundation/Buttons/Materials/Arrows/M_CycleButton_Arrow_Left
     tiling=NoTile
     imageSize=<24 24>
     margin=<ltrb 0 0 0 0>
     outlineSettings {
      bUseBrushTransparency=false
      roundingType=HalfHeightRadius
      width=0
      color {colorUseRule=UseColor_Specified, specifiedColor=<rgba 0 0 0 0>}
      cornerRadii=<0 0 0 0>
     }
     tintColor {colorUseRule=UseColor_Specified, specifiedColor=<rgba 1 1 1 1>}
     uVRegion {bIsValid=false, max=<0 0>, min=<0 0>}
    }
   }
   slot {class=HorizontalBoxSlot, properties {horizontalAlignment=HAlign_Center, verticalAlignment=VAlign_Center}}
  }
  CommonActionWidget_65 {
   widgetClassPath=&CommonActionWidget
   bIsVariable=true
   parent=HorizontalBox_3
   widget=CommonActionWidget_65
   properties {inputActions [{dataTable=&InputActionDataTable, rowName=Back}]}
   slot {class=HorizontalBoxSlot, properties {horizontalAlignment=HAlign_Center, verticalAlignment=VAlign_Center}}
  }
  InputActionWidget {widgetClassPath=&CommonActionWidget, bInherited=true, widget=null}
 ]
}
