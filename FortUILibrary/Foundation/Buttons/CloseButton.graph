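# CloseButton is a CommonButtonLegacy widget that shows a fixed-size square close button with an X icon.
# A blur post-buffer update, UIKit backplate and padded border layer under a row holding the inherited input-action glyph and icon.
# HandleSize applies default or overridden label font size and border padding, and SetText stores and displays a description.
# The label's wrapper starts collapsed, and the EventGraph is empty, so neither function is called here.
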WidgetBlueprint /FortUILibrary/Foundation/Buttons/CloseButton
graphs [
 EventGraph {links [], nodes []}
 HandleSize {nodes [
   HandleSize() {is=title K2Node_FunctionEntry}
   n1 = Select(Option 0:int=22, Option 1:int=Get FontSize() -> (FontSize:int=0), Index=Get FontSizeOveride() -> (FontSizeOveride:bool)) -> (ReturnValue:int)
   > SetFont(
    self=Get Text_ButtonAction() -> (Text_ButtonAction:Common Text Object Reference)
    InFontInfo=MakeSlateFontInfo!(
     FontObject:Object Reference=&BurbankSmall
     FontMaterial:Object Reference
     OutlineSettings:Font Outline Settings Structure=(OutlineSize=0,bMiteredCorners=False,bSeparateFillAlpha=False,bApplyOutlineToDropShadows=False,OutlineMaterial=None,OutlineColor=(R=0,G=0,B=0,A=1))
     TypefaceFontName:Name=BLACK
     Size:float="To Float (Integer)"(InInt=@n1)
     LetterSpacing:int=0
     SkewAmount:float=0
     bForceMonospaced:bool=false
     bMaterialIsStencil:bool=false
     MonospacedWidth:float=1
    ) -> (ReturnValue:Slate Font Info Structure)
   )
   > SetPadding!(
    execute
    self:Border Object Reference=Get Border_Container() -> (Border_Container:Border Object Reference)
    InPadding:Margin Structure=Select(
     Option 0:Margin Structure=Make Margin(Left=48, Top=14, Right=48, Bottom=14)
     Option 1:Margin Structure=Get Padding Overide() -> (Padding Overide:Margin Structure)
     Index=Get PaddingOveride() -> (PaddingOveride:bool)
    ) -> (ReturnValue:Margin Structure)
   ) -> (then)
  ]}
 SetText {nodes [
   n1 = SetText!() -> (then, Text:Text) {is=title K2Node_FunctionEntry}
   > n2 = Set Button Description(Button Description:Text=@n1) -> (Output_Get:Text)
   > "SetText (Text)"(self=Get Text_ButtonAction(), InText=@n2)
  ]}
]
variables [
 Button Description {type=string, default="{TEXT}"}
 FontSize {type=integer, default=22}
 FontSizeOveride {type=boolean}
 Padding Overide {type=Margin, default=<ltrb 0 0 0 0>}
 PaddingOveride {type=boolean, default=true}
]
widgets {
 info {inheritedWidgetCount=1, parentClass=&CommonButtonLegacy, rootWidgetClass=&SizeBox, widgetCount=11}
 widgets [
  SizeBox_Control {
   widgetClassPath=&SizeBox
   bIsVariable=true
   widget=SizeBox_Control
   properties {
    bOverride_HeightOverride=true
    bOverride_MinDesiredHeight=true
    bOverride_MinDesiredWidth=true
    bOverride_WidthOverride=true
    heightOverride=110
    minDesiredHeight=110
    minDesiredWidth=110
    widthOverride=110
   }
  }
  Overlay_29 {widgetClassPath=&Overlay, parent=SizeBox_Control, widget=Overlay_29, slot {class=SizeBoxSlot}}
  PostBufferUpdate_140 {
   widgetClassPath=/Script/UMG.PostBufferUpdate
   parent=Overlay_29
   widget=PostBufferUpdate_140
   properties {updateBufferInfos [{
      bufferToUpdate=ESlatePostRT_0
      postParamUpdater=&self:WidgetTree.PostBufferUpdate_140.PostBufferBlurUpdater_0
     }]}
   slot {class=OverlaySlot, properties {horizontalAlignment=HAlign_Fill, verticalAlignment=VAlign_Fill}}
  }
  UIKit_Backplate {
   widgetClassPath=&WBP_UIKit_Backplate
   parent=Overlay_29
   widget=UIKit_Backplate
   properties {brightness=Medium, cornerRadius=Interactive Component}
   slot {class=OverlaySlot, properties {horizontalAlignment=HAlign_Fill, verticalAlignment=VAlign_Fill}}
  }
  Border_Container {
   widgetClassPath=&Border
   bIsVariable=true
   parent=Overlay_29
   widget=Border_Container
   properties {
    horizontalAlignment=HAlign_Center
    verticalAlignment=VAlign_Center
    background {
     bIsDynamicallyLoaded=false
     drawAs=NoDrawType
     imageType=NoImage
     mirroring=NoMirror
     resourceName=None
     resourceObject=null
     tiling=NoTile
     imageSize=<32 32>
     margin=<ltrb 0 0 0 0>
     outlineSettings {
      bUseBrushTransparency=false
      roundingType=HalfHeightRadius
      width=0
      color {colorUseRule=UseColor_Specified, specifiedColor=<rgba 0 0 0 0>}
      cornerRadii=<0 0 0 0>
     }
     tintColor {colorUseRule=UseColor_Specified, specifiedColor=<rgba 1 1 1 1>}
     uVRegion {bIsValid=false, max=<0 0>, min=<0 0>}
    }
    padding=<ltrb 12 12 12 12>
   }
   slot {class=OverlaySlot, properties {horizontalAlignment=HAlign_Fill, verticalAlignment=VAlign_Fill}}
  }
  HorizontalBox_0 {
   widgetClassPath=&HorizontalBox
   parent=Border_Container
   widget=HorizontalBox_0
   slot {class=BorderSlot, properties {horizontalAlignment=HAlign_Center, verticalAlignment=VAlign_Center, padding=<ltrb 12 12 12 12>}}
  }
  InputActionWidget {
   widgetClassPath=&CommonActionWidget
   bInherited=true
   bIsVariable=true
   parent=HorizontalBox_0
   widget=InputActionWidget
   slot {class=HorizontalBoxSlot, properties {horizontalAlignment=HAlign_Center, verticalAlignment=VAlign_Center, padding=<ltrb 8 0 8 0>}}
  }
  Border_FontColorChange {
   widgetClassPath=&Border
   parent=HorizontalBox_0
   widget=Border_FontColorChange
   properties {
    horizontalAlignment=HAlign_Center
    visibility=Collapsed
    background {
     bIsDynamicallyLoaded=false
     drawAs=NoDrawType
     imageType=NoImage
     mirroring=NoMirror
     resourceName=None
     resourceObject=null
     tiling=NoTile
     imageSize=<32 32>
     margin=<ltrb 0 0 0 0>
     outlineSettings {
      bUseBrushTransparency=false
      roundingType=HalfHeightRadius
      width=0
      color {colorUseRule=UseColor_Specified, specifiedColor=<rgba 0 0 0 0>}
      cornerRadii=<0 0 0 0>
     }
     tintColor {colorUseRule=UseColor_Specified, specifiedColor=<rgba 1 1 1 1>}
     uVRegion {bIsValid=false, max=<0 0>, min=<0 0>}
    }
    padding=<ltrb 0 0 0 0>
   }
   slot {class=HorizontalBoxSlot, properties {horizontalAlignment=HAlign_Center, verticalAlignment=VAlign_Center}}
  }
  Text_ButtonAction {
   widgetClassPath=&CommonTextBlock
   bIsVariable=true
   parent=Border_FontColorChange
   widget=Text_ButtonAction
   properties {
    justification=Center
    text=ACTION
    textTransformPolicy=ToUpper
    colorAndOpacity {colorUseRule=UseColor_Specified, specifiedColor=<rgba 0.0307 0.8148 0.9911 1>}
    font {
     bForceMonospaced=false
     bMaterialIsStencil=false
     fontMaterial=null
     fontObject=&BurbankSmall
     letterSpacing=0
     monospacedWidth=1
     size=16
     skewAmount=0
     typefaceFontName=Black
     outlineSettings {
      bApplyOutlineToDropShadows=false
      bMiteredCorners=false
      bSeparateFillAlpha=false
      outlineMaterial=null
      outlineSize=0
      outlineColor=<rgba 0 0 0 1>
     }
    }
    shadowOffset=<0 0>
   }
   slot {class=BorderSlot, properties {horizontalAlignment=HAlign_Center, padding=<ltrb 0 0 0 0>}}
  }
  SizeBox_0 {
   widgetClassPath=&SizeBox
   parent=HorizontalBox_0
   widget=SizeBox_0
   properties {bOverride_HeightOverride=true, bOverride_WidthOverride=true, heightOverride=50, widthOverride=50}
   slot {class=HorizontalBoxSlot, properties {size {sizeRule=Fill, value=1}}}
  }
  CloseIcon {
   widgetClassPath=&Image
   bIsVariable=true
   parent=SizeBox_0
   widget=CloseIcon
   properties {
    brush {
     bIsDynamicallyLoaded=false
     drawAs=Image
     imageType=NoImage
     mirroring=NoMirror
     resourceName=None
     resourceObject=/Game/UI/Foundation/Textures/Icons/Checks/T-Icon-X-128
     tiling=NoTile
     imageSize=<128 128>
     margin=<ltrb 0 0 0 0>
     outlineSettings {
      bUseBrushTransparency=false
      roundingType=HalfHeightRadius
      width=0
      color {colorUseRule=UseColor_Specified, specifiedColor=<rgba 0 0 0 0>}
      cornerRadii=<0 0 0 0>
     }
     tintColor {colorUseRule=UseColor_Specified, specifiedColor=<rgba 1 1 1 1>}
     uVRegion {bIsValid=false, max=<0 0>, min=<0 0>}
    }
    colorAndOpacity=<rgba 1 1 1 0.8>
   }
   slot {class=SizeBoxSlot}
  }
 ]
}
