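# Icon-and-text button widget for the replay UI, derived from IconTextButtonSpectatorBase, with an input action prompt below the label.
# PreConstruct applies flags to hide image, text or prompt and adjusts box size.
# Construct and input-change events swap mouse/keyboard and controller styles; label updates prefer the action's display text over ButtonText.
# Hover tints icon and text orange unless animations are ignored; SetInputAction resolves names through ReplayInputActionDataTable.
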WidgetBlueprint /FortUILibrary/Foundation/Buttons/IconTextButtonReplay
animations [OnHover {
  binding_count=2
  length_seconds=0.1
  object=&self:OnHover
  sequence {bindings [
    Icon {
     id=#1
     possessed_class=Image
     tracks [{
       type=MovieSceneColorTrack
       display_name=Color and Opacity
       sections [{
         type=MovieSceneColorSection
         end_frame=2
         start_frame=0
         channels [
          Color.R {type=MovieSceneScriptingFloatChannel, const [1,2]}
          Color.G {type=MovieSceneScriptingFloatChannel, keys [[0,1],[2,0.413]]}
          Color.B {type=MovieSceneScriptingFloatChannel, keys [[0,1],[2,0]]}
          Color.A {type=MovieSceneScriptingFloatChannel, const [1,2]}
         ]
        }]
      }]
    }
    Text {
     id=#2
     possessed_class=CommonTextBlock
     tracks [{
       type=MovieSceneColorTrack
       display_name=Color and Opacity
       sections [{
         type=MovieSceneColorSection
         end_frame=2
         start_frame=0
         channels [
          Color.R {type=MovieSceneScriptingFloatChannel, const [1,2]}
          Color.G {type=MovieSceneScriptingFloatChannel, keys [[0,1],[2,0.413]]}
          Color.B {type=MovieSceneScriptingFloatChannel, keys [[0,1],[2,0]]}
          Color.A {type=MovieSceneScriptingFloatChannel, const [1,2]}
         ]
        }]
      }]
    }
   ]}
 }]
graphs [
 EventGraph {nodes [
   n1 = "On Input Method Changed (InputActionWidget)"!() -> (OutputDelegate, then, bUsingGamepad:bool) {is=title K2Node_ComponentBoundEvent}
   > UpdateStyle!(execute, self, UsingGamepad:bool=@n1.bUsingGamepad) -> (then)
   Event On Hovered()
   > n2 = Branch(Condition=Get bIgnoreAnimations() -> (bIgnoreAnimations:bool))
   PlayAnimationForward(execute=@n2.else, InAnimation=Get OnHover() -> (OnHover:Widget Animation Object Reference))
   n3 = Event SetText!() -> (OutputDelegate, then, Text:"Text (by ref)") {is=title K2Node_Event}
   > n4 = Set ButtonText(ButtonText:Text=@n3.Text) -> (Output_Get:Text)
   n5 = Get Text!() -> (Text:Common Text Object Reference)
   > "SetText (Text)"(self=@n5, InText=@n4)
   > SetVisibility(self=@n5, InVisibility=Select(Option 0=SelfHitTestInvisible, Option 1=Collapsed, Index=TextIsEmpty(InText=@n3.Text)))
   n6 = Event SetTabLabelInfo!() -> (OutputDelegate, then, TabLabelInfo:"Fort Tab Button Label Info Structure (by ref)") {is=title K2Node_Event}
   > n7 = Sequence()
   n8 = Break Fort Tab Button Label Info!(FortTabButtonLabelInfo:"Fort Tab Button Label Info Structure (by ref)"=@n6.TabLabelInfo) -> (DisplayName:Text, IconBrush:Slate Brush Structure, IconResource:Soft Object Reference) {is=title K2Node_BreakStruct}
   > SetText!(execute, self:Icon Text Button Spectator Base Object Reference, Text:"Text (by ref)"=@n8) -> (then)
   SetIcon!(execute=@n7.then_1, self, IconBrush:Slate Brush Structure=@n8.IconBrush) -> (then)
   Event On Unhovered()
   > n9 = Branch(Condition=Get bIgnoreAnimations())
   PlayAnimationReverse(execute=@n9.else, InAnimation=Get OnHover())
   Event OnTriggeredInputActionChanged!() -> (OutputDelegate, then, NewTriggeredAction:"Data Table Row Handle Structure (by ref)") {is=title K2Node_Event}
   n10 = Get CommonInputSubsystem!() -> (ReturnValue:Common Input Object Reference) {is=title K2Node_GetBlueprintContextEx}
   > n11 = Is Valid(InputObject=@n10)
   > UpdateTextAndStyle!(execute, self, bUsingGamepad:bool="Equal (Enum)"(A=GetCurrentInputType(self=@n10), B=Gamepad)) -> (then)
   UpdateTextAndStyle(execute=@n11.Is Not Valid, bUsingGamepad=IsInputType!(InputType:ECommonInputType Enum=Gamepad, IsNot:bool=false) -> (Result:bool))
   Event Construct()
   > InitializeButton()
   n12 = Get CommonInputSubsystem()
   > n13 = Is Valid(InputObject=@n12)
   > UpdateTextAndStyle(bUsingGamepad="Equal (Enum)"(A=GetCurrentInputType(self=@n12), B=Gamepad))
   UpdateTextAndStyle(execute=@n13.Is Not Valid, bUsingGamepad=IsInputType(InputType=Gamepad, IsNot=false))
   Event PreConstruct()
   > SetBrush(self=Get Icon() -> (Icon:Image Object Reference), InBrush=Get IconBrush() -> (IconBrush:Slate Brush Structure))
   > "SetText (Text)"(self=Get Text(), InText=Get ButtonText() -> (ButtonText:Text))
   > n14 = Branch(Condition=Get bShouldHideImage() -> (bShouldHideImage:bool))
   > SetVisibility(self=Get Icon(), InVisibility=Collapsed)
   > n15 = SetVisibility(self=Get HeightOverride() -> (HeightOverride:Spacer Object Reference), InVisibility=SelfHitTestInvisible)
   n16 = Branch(execute=[@n15.then, @n23.then], Condition=Get bHideText() -> (bHideText:bool))
   > n17 = SetVisibility(self=Get Text(), InVisibility=Collapsed)
   n18 = Branch(execute=[@n17.then, @n22.then], Condition=Get ShouldHideInputAction() -> (ShouldHideInputAction:bool))
   > SetVisibility(self=Get InputActionWidget() -> (InputActionWidget:Common Action Widget Object Reference), InVisibility=Collapsed)
   > n19 = SetMinDesiredHeight!(execute, self:Size Box Object Reference=Get Box() -> (Box:Size Box Object Reference), InMinDesiredHeight:float=40) -> (then)
   n20 = Branch(execute=[@n19.then, @n21.then], Condition=Get bUseCustomWidth() -> (bUseCustomWidth:bool))
   > SetWidthOverride(self=Get Box(), InWidthOverride=Get CustomWidth() -> (CustomWidth:float))
   ClearWidthOverride(execute=@n20.else, self:Size Box Object Reference=Get Box())
   SetVisibility(execute=@n18.else, self=Get InputActionWidget(), InVisibility=Visible)
   > n21 = SetMinDesiredHeight(self=Get Box(), InMinDesiredHeight=79)
   n22 = SetVisibility(execute=@n16.else, self=Get Text(), InVisibility=SelfHitTestInvisible)
   SetVisibility(execute=@n14.else, self=Get Icon(), InVisibility=Visible)
   > n23 = SetVisibility(self=Get HeightOverride(), InVisibility=Collapsed)
  ]}
 InitializeButton {nodes [
   InitializeButton() {is=title K2Node_FunctionEntry}
   > Set MouseKeyboardStyle(MouseKeyboardStyle:Common Button Style Class Reference=Get Style() -> (Style:Common Button Style Class Reference)) -> (Output_Get:Common Button Style Class Reference)
   > n1 = Is Valid(InputObject=GetBrushResource!(Brush:"Slate Brush Structure (by ref)"=Get IconBrush()) -> (ReturnValue:Object Reference))
   > SetBrush(self=Get Icon(), InBrush=Get IconBrush())
   SetVisibility(execute=@n1.Is Not Valid, self=Get Icon(), InVisibility=Collapsed)
  ]}
 SetControllerStyle {nodes [
   n1 = SetControllerStyle!() -> (then, ControllerInputStyle:Common Button Style Class Reference) {is=title K2Node_FunctionEntry}
   > Set ControllerInputStyle(ControllerInputStyle:Common Button Style Class Reference=@n1) -> (Output_Get:Common Button Style Class Reference)
   n2 = Get CommonInputSubsystem()
   > UpdateStyle(UsingGamepad="Equal (Enum)"(A=GetCurrentInputType(self=@n2), B=Gamepad))
  ]}
 SetIcon {nodes [
   n1 = SetIcon!() -> (then, IconBrush:Slate Brush Structure) {is=title K2Node_FunctionEntry}
   > n2 = Set IconBrush(IconBrush:Slate Brush Structure=@n1) -> (Output_Get:Slate Brush Structure)
   > SetBrush(self=Get Icon(), InBrush=@n2)
  ]}
 SetInputAction {nodes [
   n1 = SetInputAction!() -> (then, InputActionName:Name) {is=title K2Node_FunctionEntry}
   > SetInputAction!(
    execute
    self:Common Action Widget Object Reference=Get InputActionWidget()
    InputActionRow_DataTable:Data Table Object Reference=/FortUILibrary/Input/ReplayInputActionDataTable
    InputActionRow_RowName:Name=@n1
   ) -> (then) {is=title K2Node_CallFunction}
  ]}
 SetMouseKeyboardStyle {nodes [
   n1 = SetMouseKeyboardStyle!() -> (then, ControllerInputStyle:Common Button Style Class Reference) {is=title K2Node_FunctionEntry}
   > Set MouseKeyboardStyle(MouseKeyboardStyle=@n1)
   n2 = Get CommonInputSubsystem()
   > UpdateStyle(UsingGamepad="Equal (Enum)"(A=GetCurrentInputType(self=@n2), B=Gamepad))
  ]}
 UpdateStyle {nodes [
   n1 = UpdateStyle!() -> (then, UsingGamepad:bool=~) {is=title K2Node_FunctionEntry}
   > n2 = Set LocUsingGamepad(LocUsingGamepad:bool=@n1) -> (Output_Get:bool)
   > n3 = Branch(Condition=@n2)
   > n4 = SetStyle!(
    execute
    self:Common Button Base Object Reference
    InStyle:Common Button Style Class Reference=Get ControllerInputStyle() -> (ControllerInputStyle:Common Button Style Class Reference)
   ) -> (then)
   Branch(execute=[@n4.then, @n5.then], Condition=Get ShouldHideInputAction())
   > SetVisibility(self=Get InputActionWidget(), InVisibility=Collapsed)
   n5 = SetStyle(execute=@n3.else, InStyle=Get MouseKeyboardStyle() -> (MouseKeyboardStyle:Common Button Style Class Reference))
  ]}
 UpdateText {nodes [
   UpdateText() {is=title K2Node_FunctionEntry}
   n1 = GetDisplayText!(self:Common Action Widget Object Reference=Get InputActionWidget()) -> (ReturnValue:Text)
   n2 = TextIsEmpty(InText=@n1)
   > SetText(Text=Select(Option 0:Text=Get ButtonText(), Option 1:Text=@n1, Index=NOT Boolean(A=@n2)) -> (ReturnValue:Text))
  ]}
 UpdateTextAndStyle {nodes [
   n1 = UpdateTextAndStyle!() -> (then, bUsingGamepad:bool=~) {is=title K2Node_FunctionEntry}
   > UpdateStyle(UsingGamepad=@n1) {is=title K2Node_CallFunction}
   > UpdateText() {is=title K2Node_CallFunction}
  ]}
]
variables [
 bHideText {type=boolean, default=true}
 bIgnoreAnimations {type=boolean}
 bShouldHideImage {type=boolean}
 bUseCustomWidth {type=boolean}
 ButtonText {type=string, default=""}
 ControllerInputStyle {type=Class, default=null}
 CustomWidth {type=number, default=56}
 DefaultColor {type=LinearColor, default=<rgba 0.0176 0.0232 0.0331 0.5>}
 HoveredColor {type=LinearColor, default=<rgba 1 0.4125 0 1>}
 IconBrush {
  type=SlateBrush
  default {
   bIsDynamicallyLoaded=false
   drawAs=Image
   imageType=NoImage
   mirroring=NoMirror
   resourceName=None
   resourceObject=null
   tiling=NoTile
   imageSize=<32 32>
   margin=<ltrb 0 0 0 0>
   outlineSettings {
    bUseBrushTransparency=false
    roundingType=HalfHeightRadius
    width=0
    color {colorUseRule=UseColor_Specified, specifiedColor=<rgba 0 0 0 0>}
    cornerRadii=<0 0 0 1>
   }
   tintColor {colorUseRule=UseColor_Specified, specifiedColor=<rgba 1 1 1 1>}
   uVRegion {bIsValid=false, max=<0 0>, min=<0 0>}
  }
 }
 MouseKeyboardStyle {type=Class, default=/FortUILibrary/Foundation/Buttons/ButtonStyles/Primary/ButtonStyle-Primary-M.ButtonStyle-Primary-M_C}
 SelectedColor {type=LinearColor, default=<rgba 1 0.4125 0 1>}
 ShouldHideInputAction {type=boolean}
]
widgets {
 info {inheritedWidgetCount=1, parentClass=/Script/FortniteUI.IconTextButtonSpectatorBase, rootWidgetClass=&SizeBox, widgetCount=8}
 widgets [
  Box {
   widgetClassPath=&SizeBox
   bIsVariable=true
   widget=Box
   properties {
    bOverride_MaxDesiredWidth=true
    bOverride_MinDesiredHeight=true
    bOverride_MinDesiredWidth=true
    heightOverride=40
    maxDesiredWidth=56
    minDesiredHeight=79
    minDesiredWidth=56
    widthOverride=38.537
   }
  }
  Overlay {
   widgetClassPath=&Overlay
   bIsVariable=true
   parent=Box
   widget=Overlay
   properties {bOverride_Cursor=true, visibility=Visible}
   slot {class=SizeBoxSlot}
  }
  VerticalBox_0 {
   widgetClassPath=&VerticalBox
   parent=Overlay
   widget=VerticalBox_0
   slot {class=OverlaySlot, properties {horizontalAlignment=HAlign_Fill, verticalAlignment=VAlign_Fill}}
  }
  HorizontalBox_1 {
   widgetClassPath=&HorizontalBox
   parent=VerticalBox_0
   widget=HorizontalBox_1
   slot {class=VerticalBoxSlot, properties {horizontalAlignment=HAlign_Center, padding=<ltrb 5 5 5 5>}}
  }
  Icon {
   widgetClassPath=&Image
   bIsVariable=true
   parent=HorizontalBox_1
   widget=Icon
   properties {brush {
     bIsDynamicallyLoaded=false
     drawAs=Image
     imageType=FullColor
     mirroring=NoMirror
     resourceName=None
     resourceObject=/Game/UI/Foundation/Textures/Icons/Replay/T-Icon-Replay-CameraSettings-L
     tiling=NoTile
     imageSize=<32 32>
     margin=<ltrb 0 0 0 0>
     outlineSettings {
      bUseBrushTransparency=false
      roundingType=HalfHeightRadius
      width=0
      color {colorUseRule=UseColor_Specified, specifiedColor=<rgba 0 0 0 0>}
      cornerRadii=<0 0 0 1>
     }
     tintColor {colorUseRule=UseColor_Specified, specifiedColor=<rgba 1 1 1 1>}
     uVRegion {bIsValid=false, max=<0 0>, min=<0 0>}
    }}
   slot {class=HorizontalBoxSlot, properties {horizontalAlignment=HAlign_Center, verticalAlignment=VAlign_Center}}
  }
  Text {
   widgetClassPath=&CommonTextBlock
   bIsVariable=true
   parent=HorizontalBox_1
   widget=Text
   properties {
    justification=Center
    text="{Label}"
    font {
     bForceMonospaced=false
     bMaterialIsStencil=false
     fontMaterial=null
     fontObject=&HeadingNow
     letterSpacing=0
     monospacedWidth=1
     size=13.5
     skewAmount=0
     typefaceFontName=74 Regular
     outlineSettings {
      bApplyOutlineToDropShadows=false
      bMiteredCorners=false
      bSeparateFillAlpha=false
      outlineMaterial=null
      outlineSize=0
      outlineColor=<rgba 0 0 0 1>
     }
    }
    shadowOffset=<0 0>
   }
   slot {class=HorizontalBoxSlot, properties {horizontalAlignment=HAlign_Left, verticalAlignment=VAlign_Center}}
  }
  HeightOverride {
   widgetClassPath=&Spacer
   bIsVariable=true
   parent=HorizontalBox_1
   widget=HeightOverride
   properties {visibility=Collapsed, size=<0 32>}
   slot {class=HorizontalBoxSlot}
  }
  InputActionWidget {
   widgetClassPath=&CommonActionWidget
   bInherited=true
   bIsVariable=true
   parent=VerticalBox_0
   widget=InputActionWidget
   slot {class=VerticalBoxSlot, properties {horizontalAlignment=HAlign_Center, verticalAlignment=VAlign_Bottom, padding=<ltrb 0 5 0 0>}}
  }
 ]
}
