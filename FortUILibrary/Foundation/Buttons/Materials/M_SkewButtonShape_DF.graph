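# Translucent UI material that draws a skewed button shape with a drop shadow and hover tint from a distance-field texture.
# AAThreshold cuts an antialiased shape at ShapeSize, filled with ShapeColor blended toward HoverColor by HoverAmount.
# A second sample, shifted by ShadowOffset times the UI pixel size, forms the shadow mask, scaled by ShadowColor alpha.
# The shape mask layers fill over shadow color, and the maximum of both masks gives combined coverage.
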Material /FortUILibrary/Foundation/Buttons/Materials/M_SkewButtonShape_DF
graph {
 properties {
  bUsedWithStaticMesh=false
  blendMode=BLEND_Translucent
  materialDomain=MD_UI
  pixelDepthOffsetMode=PDOM_Legacy
  usedShadingModels=Default Lit
 }
 nodes [
  ComponentMask() {g=true, r=true}
  n1 = VectorParameter() {parameterName=ShadowColor, defaultValue=<rgba 0.0052 0.0052 0.0056 0.75>}
  n2 = ScalarParameter() {defaultValue=0.2, parameterName=ShapeSize}
  n3 = MaterialFunctionCall!(
   Threshold=@n2
   Gradient=TextureSampleParameter2D() {parameterName=Texture, texture=/FortUILibrary/Foundation/Buttons/Textures/T_SkewButtonShape_B_DF}.R
   Filter Width Override
   Filter Width Fudge Factor
  ) {materialFunction=/Game/UI/Materials/AAThreshold}
  LinearInterpolate(
   A=@n1.RGB
   B=LinearInterpolate(
    A=VectorParameter() {parameterName=ShapeColor, defaultValue=<rgba 1 1 1 1>}.RGB
    B=VectorParameter() {parameterName=HoverColor, defaultValue=<rgba 1 0.9444 0.4103 1>}.RGB
    Alpha=ScalarParameter() {parameterName=HoverAmount}
   )
   Alpha=@n3.Result
  )
  n4 = MaterialFunctionCall() {materialFunction=&GetUserInterfaceUV}
  n5 = MaterialFunctionCall(Float2=@n4.Pixel Size) {materialFunction=&BreakOutFloat2Components}
  n6 = VectorParameter() {parameterName=ShadowOffset, defaultValue=<rgba -0.1 -0.1 0 1>}
  n7 = Multiply(A=@n5.R, B=@n6.R)
  n8 = Multiply(A=@n5.G, B=@n6.G)
  n9 = TextureSampleParameter2D(UVs=Add(A=@n4.Normalized UV, B=AppendVector(A=@n7, B=@n8))) {parameterName=Texture, texture=&T_SkewButtonShape_B_DF}
  Max(
   A=@n3.Result
   B=Multiply(A=MaterialFunctionCall(Threshold=@n2, Gradient=@n9.R) {materialFunction=&AAThreshold}.Result, B=@n1.A)
  )
  TextureCoordinate()
 ]
}
