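# UI material for an external link button icon, translucent in the UI domain.
# It samples the external link icon texture as an alpha mask with clamped addressing.
# A SmoothStep over a threshold band, built from a lower edge plus a fixed width, remaps the red channel to control the softness of the icon's edge.
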Material /FortUILibrary/Foundation/Buttons/Materials/M_UI_Icon_ExternalLink
graph {
 properties {
  bUsedWithStaticMesh=false
  blendMode=BLEND_Translucent
  materialDomain=MD_UI
  pixelDepthOffsetMode=PDOM_Legacy
  usedShadingModels=Default Lit
 }
 nodes [
  Constant() {r=1}
  n1 = Constant() {r=0.3}
  n2 = TextureSample() {
   samplerSource=SSM_Clamp_WorldGroupSettings
   samplerType=SAMPLERTYPE_Alpha
   texture=/FortUILibrary/Foundation/Buttons/Textures/T_Icon_ExternalLink
  }
  SmoothStep(Min=@n1, Max=Add(A=@n1, B=Constant() {r=0.5}), Value=@n2.R)
 ]
}
