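# PanelButton is a CommonButtonLegacy-based UI button with a content slot and an overlaid input action icon.
# PreConstruct positions the icon and sizes or clears its rim brush from variables, and at design time hides an empty slot.
# Construct caches the current style for mouse/keyboard, then input-method or triggered-action changes swap between it and ControllerInputStyle.
# Press and release broadcast HoldActionStarted and HoldActionEnded.
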WidgetBlueprint /FortUILibrary/Foundation/Buttons/PanelButton
graphs [
 EventGraph {nodes [
   n1 = "On Input Method Changed (InputActionWidget)"!() -> (OutputDelegate, then, bUsingGamepad:bool) {is=title K2Node_ComponentBoundEvent}
   > UpdateStyle!(execute, self, bUsingGamepad:bool=@n1.bUsingGamepad) -> (then)
   Event On Released() {is=title K2Node_Event}
   > Call Hold Action Ended!(execute:Exec, self:Panel Button Object Reference) -> (then:Exec) {is=title K2Node_CallDelegate}
   Event On Pressed() {is=title K2Node_Event}
   > Call Hold Action Started!(execute:Exec, self:Panel Button Object Reference) -> (then:Exec) {is=title K2Node_CallDelegate}
   Event OnTriggeredInputActionChanged!() -> (OutputDelegate, then, NewTriggeredAction:"Data Table Row Handle Structure (by ref)") {is=title K2Node_Event}
   > UpdateTextAndStyle()
   Event Construct()
   > InitializeButton()
   n2 = Event PreConstruct()
   > n3 = Sequence()
   > Branch(Condition=@n2.IsDesignTime)
   n4 = Get ContentSlot() -> (ContentSlot:Named Slot Object Reference)
   > SetVisibility(
    self=@n4
    InVisibility=Select(Option 0=Collapsed, Option 1=SelfHitTestInvisible, Index=HasAnyChildren!(self:Panel Widget Object Reference=@n4) -> (ReturnValue:bool))
   )
   UpdateInputActionLayout(execute=@n3.then_1)
   Event OnActionComplete() {is=title K2Node_Event}
   Event OnActionProgress!() -> (OutputDelegate, then, HeldPercent:float) {is=title K2Node_Event}
   Event On Hovered()
   Event On Unhovered()
  ]}
 HoldActionEnded {nodes [HoldActionEnded() {is=title K2Node_FunctionEntry}]}
 HoldActionStarted {nodes [HoldActionStarted() {is=title K2Node_FunctionEntry}]}
 InitializeButton {nodes [
   InitializeButton() {is=title K2Node_FunctionEntry}
   > Set MouseKeyboardStyle(MouseKeyboardStyle:Common Button Style Class Reference=Get Style() -> (Style:Common Button Style Class Reference)) -> (Output_Get:Common Button Style Class Reference)
   > UpdateTextAndStyle()
  ]}
 UpdateInputActionLayout {nodes [
   UpdateInputActionLayout() {is=title K2Node_FunctionEntry}
   n1 = Get InputActionWidget() -> (InputActionWidget:Common Action Widget Object Reference)
   > SetRenderTranslation(self=@n1, Translation=Get InputActionRenderTranslation() -> (InputActionRenderTranslation:Vector 2D Structure))
   > Is Valid(InputObject=@n1)
   > n2 = Cast To OverlaySlot!(execute, Object=Get Slot!(self:Widget Object Reference=@n1) -> (Slot:Panel Slot Object Reference)) -> (then, CastFailed, AsOverlay Slot:Overlay Slot Object Reference) {is=title K2Node_DynamicCast}
   > SetHorizontalAlignment(
    self=@n2
    InHorizontalAlignment=Get InputActionHorizontalAlignment() -> (InputActionHorizontalAlignment:EHorizontalAlignment Enum=HAlign_Fill)
   )
   > SetVerticalAlignment(self=@n2, InVerticalAlignment=Get InputActionVerticalAlignment() -> (InputActionVerticalAlignment:EVerticalAlignment Enum=VAlign_Fill))
   > n3 = Set RimBrush(RimBrush:Slate Brush Structure=Get IconRimBrush!(self:Common Action Widget Object Reference=@n1) -> (IconRimBrush:Slate Brush Structure)) -> (Output_Get:Slate Brush Structure)
   > n4 = Set members in Slate Brush!(
    execute:Exec
    StructRef:"Slate Brush Structure (by ref)"=@n3
    ImageSize:Vector 2D Structure=Get InputActionRimBrushSize() -> (InputActionRimBrushSize:Vector 2D Structure)
   ) -> (then:Exec, StructOut:"Slate Brush Structure (by ref)") {is=title K2Node_SetFieldsInStruct}
   > SetIconRimBrush!(execute, self:Common Action Widget Object Reference=@n1, InIconRimBrush:Slate Brush Structure=@n4) -> (then)
   > n5 = Branch(Condition=Get InputActionUseRimBrush() -> (InputActionUseRimBrush:bool))
   SetIconRimBrush(
    execute=@n5.else
    self=@n1
    InIconRimBrush=Make Slate Brush!(
     TintColor:Slate Color Structure=(SpecifiedColor=(R=1,G=1,B=1,A=1),ColorUseRule=UseColor_Specified)
     DrawAs:ESlateBrushDrawType Enum=NoDrawType
     Tiling:ESlateBrushTileType Enum=NoTile
     Mirroring:ESlateBrushMirrorType Enum=NoMirror
     ImageSize:Vector 2D Structure=(X=32,Y=32)
     Margin:Margin Structure=(Left=0,Top=0,Right=0,Bottom=0)
     ResourceObject:Object Reference
     OutlineSettings:Slate Brush Outline Settings Structure=(CornerRadii=(X=0,Y=0,Z=0,W=0),Color=(SpecifiedColor=(R=0,G=0,B=0,A=0),ColorUseRule=UseColor_Specified),Width=0,RoundingType=HalfHeightRadius,bUseBrushTransparency=False)
    ) -> (SlateBrush:Slate Brush Structure) {is=title K2Node_MakeStruct}
   )
  ]}
 UpdateStyle {nodes [
   n1 = UpdateStyle!() -> (then, bUsingGamepad:bool=~) {is=title K2Node_FunctionEntry}
   > n2 = Branch(Condition=@n1)
   > SetStyle!(
    execute
    self:Common Button Base Object Reference
    InStyle:Common Button Style Class Reference=Get ControllerInputStyle() -> (ControllerInputStyle:Common Button Style Class Reference)
   ) -> (then)
   SetStyle(execute=@n2.else, InStyle=Get MouseKeyboardStyle() -> (MouseKeyboardStyle:Common Button Style Class Reference))
  ]}
 UpdateTextAndStyle {nodes [
   UpdateTextAndStyle() {is=title K2Node_FunctionEntry}
   n1 = Get CommonInputSubsystem!() -> (ReturnValue:Common Input Object Reference) {is=title K2Node_GetBlueprintContextEx}
   > UpdateStyle(bUsingGamepad="Equal (Enum)"(A=GetCurrentInputType(self=@n1), B=Gamepad)) {is=title K2Node_CallFunction}
  ]}
]
variables [
 ControllerInputStyle {type=Class, default=null}
 HoldActionEnded {}
 HoldActionStarted {}
 InputActionHorizontalAlignment {type=EHorizontalAlignment, default=HAlign_Center}
 InputActionRenderTranslation {type=Vector2D, default=<0 24>}
 InputActionRimBrushSize {type=Vector2D, default=<44 44>}
 InputActionUseRimBrush {type=boolean, default=true}
 InputActionVerticalAlignment {type=EVerticalAlignment, default=VAlign_Bottom}
 MouseKeyboardStyle {type=Class, default=/FortUILibrary/Foundation/Buttons/ButtonStyles/Primary/ButtonStyle-Primary-M.ButtonStyle-Primary-M_C}
]
widgets {
 info {inheritedWidgetCount=1, parentClass=&CommonButtonLegacy, rootWidgetClass=&Overlay, widgetCount=3}
 widgets [
  Overlay_0 {widgetClassPath=&Overlay, widget=Overlay_0}
  ContentSlot {
   widgetClassPath=&NamedSlot
   bIsVariable=true
   parent=Overlay_0
   widget=ContentSlot
   properties {bExposeOnInstanceOnly=true}
   slot {class=OverlaySlot, properties {horizontalAlignment=HAlign_Fill, verticalAlignment=VAlign_Fill}}
  }
  InputActionWidget {
   widgetClassPath=&CommonActionWidget
   bInherited=true
   bIsVariable=true
   parent=Overlay_0
   widget=InputActionWidget
   properties {
    visibility=HitTestInvisible
    iconRimBrush {
     bIsDynamicallyLoaded=false
     drawAs=Image
     imageType=NoImage
     mirroring=NoMirror
     resourceName=None
     resourceObject=/FortUILibrary/Foundation/Buttons/Materials/M-Button-Console-YellowOutline
     tiling=NoTile
     imageSize=<44 44>
     margin=<ltrb 0 0 0 0>
     outlineSettings {
      bUseBrushTransparency=false
      roundingType=HalfHeightRadius
      width=0
      color {colorUseRule=UseColor_Specified, specifiedColor=<rgba 0 0 0 0>}
      cornerRadii=<0 0 0 0>
     }
     tintColor {colorUseRule=UseColor_Specified, specifiedColor=<rgba 1 1 1 1>}
     uVRegion {bIsValid=false, max=<0 0>, min=<0 0>}
    }
    renderTransform {angle=0, scale=<1 1>, shear=<0 0>, translation=<0 24>}
   }
   slot {class=OverlaySlot, properties {horizontalAlignment=HAlign_Center, verticalAlignment=VAlign_Bottom}}
  }
 ]
}
