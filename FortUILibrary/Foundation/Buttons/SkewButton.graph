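# SkewButton is a CommonButtonLegacy-based button with a named content slot, input action icon, and shaped fill and shadow images.
# PreConstruct picks the DF or Chip material and sets its colors, deriving hover color by HSV when HoverColor alpha is zero.
# If Skew is set, PreConstruct and Construct randomly shear the shapes; name-seeded stream nodes are unconnected.
# Hover animates the material unless NoHoverColor is set, while mouse and touch overrides play press/release animations, returning Unhandled.
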WidgetBlueprint /FortUILibrary/Foundation/Buttons/SkewButton
animations [
 Hover {
  binding_count=1
  display_label=""
  length_seconds=0.1
  object=&self:Hover
  sequence {bindings [ImageButtonFill {
     id=#1
     possessed_class=Image
     tracks [{
       type=MovieSceneWidgetMaterialTrack
       display_name=Brush Material
       sections [{
         type=MovieSceneParameterSection
         end_frame=2
         start_frame=0
         channels [HoverAmount {type=MovieSceneScriptingFloatChannel, keys [[0,0],[2,1]]}]
        }]
      }]
    }]}
 }
 Pressed {
  binding_count=3
  display_label=""
  length_seconds=0.1
  object=&self:Pressed
  sequence {
   bindings [
    Overlay_0 {
     id=#2
     possessed_class=Overlay
     tracks [{
       type=MovieScene2DTransformTrack
       sections [{
         type=MovieScene2DTransformSection
         end_frame=10
         start_frame=0
         channels [
          Translation.X {type=MovieSceneScriptingFloatChannel}
          Translation.Y {type=MovieSceneScriptingFloatChannel}
          Angle {type=MovieSceneScriptingFloatChannel}
          Scale.X {type=MovieSceneScriptingFloatChannel, keys [[0,1],[10,0.8]]}
          Scale.Y {type=MovieSceneScriptingFloatChannel, keys [[0,1],[10,0.8]]}
          Shear.X {type=MovieSceneScriptingFloatChannel}
          Shear.Y {type=MovieSceneScriptingFloatChannel}
         ]
        }]
      }]
    }
    ImageShadow {
     id=#3
     possessed_class=Image
     tracks [{
       type=MovieScene2DTransformTrack
       sections [{
         type=MovieScene2DTransformSection
         end_frame=10
         start_frame=0
         channels [
          Translation.X {type=MovieSceneScriptingFloatChannel, keys [[0,8],[9.9983,4]]}
          Translation.Y {type=MovieSceneScriptingFloatChannel, keys [[0,8],[9.9983,4]]}
          Angle {type=MovieSceneScriptingFloatChannel}
          Scale.X {type=MovieSceneScriptingFloatChannel}
          Scale.Y {type=MovieSceneScriptingFloatChannel}
          Shear.X {type=MovieSceneScriptingFloatChannel}
          Shear.Y {type=MovieSceneScriptingFloatChannel}
         ]
        }]
      }]
    }
    ImageButtonFill {
     id=#4
     possessed_class=Image
     tracks [{
       type=MovieScene2DTransformTrack
       sections [{
         type=MovieScene2DTransformSection
         end_frame=10
         start_frame=0
         channels [
          Translation.X {type=MovieSceneScriptingFloatChannel, keys [[0,0],[10,4]]}
          Translation.Y {type=MovieSceneScriptingFloatChannel, keys [[0,0],[10,4]]}
          Angle {type=MovieSceneScriptingFloatChannel}
          Scale.X {type=MovieSceneScriptingFloatChannel, const [1,1]}
          Scale.Y {type=MovieSceneScriptingFloatChannel, const [1,1]}
          Shear.X {type=MovieSceneScriptingFloatChannel}
          Shear.Y {type=MovieSceneScriptingFloatChannel}
         ]
        }]
      }]
    }
   ]
   display_rate {numerator=100}
  }
 }
 Released {
  binding_count=3
  display_label=""
  length_seconds=0.7
  object=&self:Released
  sequence {
   bindings [
    Overlay_0 {
     id=#2
     possessed_class=Overlay
     tracks [{
       type=MovieScene2DTransformTrack
       sections [{
         type=MovieScene2DTransformSection
         end_frame=70
         start_frame=0
         channels [
          Translation.X {type=MovieSceneScriptingFloatChannel}
          Translation.Y {type=MovieSceneScriptingFloatChannel}
          Angle {type=MovieSceneScriptingFloatChannel}
          Scale.X {type=MovieSceneScriptingFloatChannel, keys [[0.9983,0.8],[10.9983,1]]}
          Scale.Y {type=MovieSceneScriptingFloatChannel, keys [[0.9983,0.8],[10.9983,1]]}
          Shear.X {type=MovieSceneScriptingFloatChannel}
          Shear.Y {type=MovieSceneScriptingFloatChannel}
         ]
        }]
      }]
    }
    ImageShadow {
     id=#3
     possessed_class=Image
     tracks [{
       type=MovieScene2DTransformTrack
       sections [{
         type=MovieScene2DTransformSection
         end_frame=70
         start_frame=0
         channels [
          Translation.X {type=MovieSceneScriptingFloatChannel, keys [[5.9983,4],[15.9983,8]]}
          Translation.Y {type=MovieSceneScriptingFloatChannel, keys [[5.9983,4],[15.9983,8]]}
          Angle {type=MovieSceneScriptingFloatChannel}
          Scale.X {type=MovieSceneScriptingFloatChannel}
          Scale.Y {type=MovieSceneScriptingFloatChannel}
          Shear.X {type=MovieSceneScriptingFloatChannel}
          Shear.Y {type=MovieSceneScriptingFloatChannel}
         ]
        }]
      }]
    }
    ImageButtonFill {
     id=#4
     possessed_class=Image
     tracks [{
       type=MovieScene2DTransformTrack
       sections [{
         type=MovieScene2DTransformSection
         end_frame=70
         start_frame=0
         channels [
          Translation.X {type=MovieSceneScriptingFloatChannel, keys [[0.9983,4],[10.9983,0]]}
          Translation.Y {type=MovieSceneScriptingFloatChannel, keys [[0.9983,4],[10.9983,0]]}
          Angle {type=MovieSceneScriptingFloatChannel}
          Scale.X {type=MovieSceneScriptingFloatChannel, keys [[0.9983,1],[15.9983,1.05],[41,1]]}
          Scale.Y {type=MovieSceneScriptingFloatChannel, keys [[0.9983,1],[15.9983,1.05],[41,1]]}
          Shear.X {type=MovieSceneScriptingFloatChannel}
          Shear.Y {type=MovieSceneScriptingFloatChannel}
         ]
        }]
      }]
    }
   ]
   display_rate {numerator=100}
  }
 }
]
graphs [
 EventGraph {nodes [
   Event PreConstruct()
   n1 = Get ImageButtonFill() -> (ImageButtonFill:Image Object Reference)
   > SetBrushFromMaterial(
    self=[@n1, Get ImageShadow() -> (ImageShadow:Image Object Reference)]
    Material=Select(
     Option 0:Material Interface Object Reference=/FortUILibrary/Foundation/Buttons/Materials/M_SkewButtonShape_DF
     Option 1:Material Interface Object Reference=/FortUILibrary/Foundation/Buttons/Materials/M_SkewButtonShape_Chip
     Index=Get Chip() -> (Chip:bool)
    ) -> (ReturnValue:Material Interface Object Reference)
   )
   > n2 = GetDynamicMaterial(self=@n1)
   n3 = Get ButtonColor() -> (ButtonColor:Linear Color Structure)
   > SetVectorParameterValue(self=@n2, ParameterName=ShapeColor, Value=@n3)
   n4 = RGB to HSV!(InColor:Linear Color Structure=@n3) -> (H:float, S:float, V:float, A:float)
   n5 = float * float(A=@n4.S, B=0.9)
   n6 = float * float(A=@n4.V, B=1.75)
   n7 = Get HoverColor() -> (HoverColor:Linear Color Structure)
   n8 = BreakColor(InColor=@n7)
   > n9 = SetVectorParameterValue(
    self=@n2
    ParameterName=HoverColor
    Value=Select(
     Option 0:Linear Color Structure=@n7
     Option 1:Linear Color Structure=HSV to RGB!(H:float=@n4, S:float=@n5, V:float=@n6, A:float=@n4.A) -> (ReturnValue:Linear Color Structure)
     Index="Equal (Float)"!(A:double=@n8.A, B:double=0) -> (ReturnValue:bool) {is=title K2Node_CallFunction}
    ) -> (ReturnValue:Linear Color Structure)
   )
   Branch(execute=[@n9.then, @n10.then], Condition=Get Skew() -> (Skew:bool))
   > SetRenderShear!(
    execute
    self:Widget Object Reference=Get Overlay_SkewBrushes() -> (Overlay_SkewBrushes:Overlay Object Reference)
    Shear:Vector 2D Structure=MakeVector2D(X=RandomFloatInRange(Min=-5, Max=5), Y=RandomFloatInRange(Min=-0.5, Max=0.5))
   ) -> (then)
   n10 = Event Construct()
   n11 = Event On Hovered()
   n12 = Get NoHoverColor() -> (NoHoverColor:bool)
   n13 = Branch(execute=[@n11.then, @n18.then], Condition=@n12)
   n14 = Get Hover() -> (Hover:Widget Animation Object Reference)
   PlayAnimation(execute=@n13.else, InAnimation=@n14)
   n15 = Event On Unhovered()
   n16 = Branch(execute=[@n15.then, @n17.then], Condition=@n12)
   PlayAnimation(execute=@n16.else, InAnimation=@n14, PlayMode=Reverse, PlaybackSpeed=0.3)
   n17 = Event OnMouseLeave!() -> (OutputDelegate, then, MouseEvent:"Pointer Event Structure (by ref)") {is=title K2Node_Event}
   n18 = Event OnMouseEnter()
   n19 = Self-Reference()
   n20 = MakeRandomStream!(InitialSeed:int=String To Integer(InString=GetObjectName(Object=@n19))) -> (ReturnValue:Random Stream Structure)
   RandomFloatInRangeFromStream!(Stream:"Random Stream Structure (by ref)"=@n20, Min:float=-0.5, Max:float=0.5) -> (ReturnValue:float)
   RandomFloatInRangeFromStream(Stream=@n20, Min=-5, Max=5)
   PlayAnimation(InAnimation=Get Pressed() -> (Pressed:Widget Animation Object Reference))
   Event Tick()
   Event On Clicked()
  ]}
 OnMouseButtonDown {nodes [
   OnMouseButtonDown!() -> (then, MyGeometry:Geometry Structure, MouseEvent:"Pointer Event Structure (by ref)") {is=title K2Node_FunctionEntry}
   > PlayAnimation(InAnimation=Get Pressed())
   > Return Node!(execute, ReturnValue:Event Reply Structure=Unhandled())
  ]}
 OnMouseButtonUp {nodes [
   OnMouseButtonUp!() -> (then, MyGeometry:Geometry Structure, MouseEvent:"Pointer Event Structure (by ref)") {is=title K2Node_FunctionEntry}
   > PlayAnimation(InAnimation=Get Released() -> (Released:Widget Animation Object Reference))
   > Return Node(ReturnValue=Unhandled())
  ]}
 OnTouchEnded {nodes [
   OnTouchEnded!() -> (then, MyGeometry:Geometry Structure, InTouchEvent:"Pointer Event Structure (by ref)") {is=title K2Node_FunctionEntry}
   > PlayAnimation(InAnimation=Get Released())
   > Return Node(ReturnValue=Unhandled())
  ]}
 OnTouchStarted {nodes [
   OnTouchStarted!() -> (then, MyGeometry:Geometry Structure, InTouchEvent:"Pointer Event Structure (by ref)") {is=title K2Node_FunctionEntry}
   > PlayAnimation(InAnimation=Get Pressed())
   > Return Node(ReturnValue=Unhandled())
  ]}
]
variables [
 ButtonColor {type=LinearColor, default=<rgba 0.3515 0.5583 1 1>}
 Chip {type=boolean}
 HoverColor {type=LinearColor, default=<rgba 0 0 0 0>}
 NoHoverColor {type=boolean}
 Skew {type=boolean, default=true}
]
widgets {
 info {inheritedWidgetCount=1, parentClass=&CommonButtonLegacy, rootWidgetClass=&Overlay, widgetCount=10}
 widgets [
  Overlay_1 {widgetClassPath=&Overlay, widget=Overlay_1}
  Overlay_0 {
   widgetClassPath=&Overlay
   parent=Overlay_1
   widget=Overlay_0
   properties {renderTransformPivot=<0.5 0.75>}
   slot {class=OverlaySlot, properties {horizontalAlignment=HAlign_Fill, verticalAlignment=VAlign_Center}}
  }
  ScaleBox_0 {
   widgetClassPath=&ScaleBox
   parent=Overlay_0
   widget=ScaleBox_0
   properties {stretch=Fill}
   slot {class=OverlaySlot, properties {horizontalAlignment=HAlign_Fill, verticalAlignment=VAlign_Fill}}
  }
  Overlay_SkewBrushes {widgetClassPath=&Overlay, bIsVariable=true, parent=ScaleBox_0, widget=Overlay_SkewBrushes, slot {class=ScaleBoxSlot}}
  ImageShadow {
   widgetClassPath=&Image
   bIsVariable=true
   parent=Overlay_SkewBrushes
   widget=ImageShadow
   properties {
    brush {
     bIsDynamicallyLoaded=false
     drawAs=Image
     imageType=NoImage
     mirroring=NoMirror
     resourceName=None
     resourceObject=&M_SkewButtonShape_DF
     tiling=NoTile
     imageSize=<32 32>
     margin=<ltrb 0 0 0 0>
     outlineSettings {
      bUseBrushTransparency=false
      roundingType=HalfHeightRadius
      width=0
      color {colorUseRule=UseColor_Specified, specifiedColor=<rgba 0 0 0 0>}
      cornerRadii=<0 0 0 0>
     }
     tintColor {colorUseRule=UseColor_Specified, specifiedColor=<rgba 0.0052 0.0052 0.0056 0.4>}
     uVRegion {bIsValid=false, max=<0 0>, min=<0 0>}
    }
    renderTransform {angle=0, scale=<1 1>, shear=<0 0>, translation=<8 8>}
   }
   slot {class=OverlaySlot, properties {horizontalAlignment=HAlign_Fill, verticalAlignment=VAlign_Fill}}
  }
  ImageButtonFill {
   widgetClassPath=&Image
   bIsVariable=true
   parent=Overlay_SkewBrushes
   widget=ImageButtonFill
   properties {
    brush {
     bIsDynamicallyLoaded=false
     drawAs=Image
     imageType=NoImage
     mirroring=NoMirror
     resourceName=None
     resourceObject=&M_SkewButtonShape_DF
     tiling=NoTile
     imageSize=<32 32>
     margin=<ltrb 0 0 0 0>
     outlineSettings {
      bUseBrushTransparency=false
      roundingType=HalfHeightRadius
      width=0
      color {colorUseRule=UseColor_Specified, specifiedColor=<rgba 0 0 0 0>}
      cornerRadii=<0 0 0 0>
     }
     tintColor {colorUseRule=UseColor_Specified, specifiedColor=<rgba 1 1 1 1>}
     uVRegion {bIsValid=false, max=<0 0>, min=<0 0>}
    }
    renderTransformPivot=<0.75 0.8>
   }
   slot {class=OverlaySlot, properties {horizontalAlignment=HAlign_Fill, verticalAlignment=VAlign_Fill}}
  }
  Border_0 {
   widgetClassPath=&Border
   parent=Overlay_0
   widget=Border_0
   properties {
    horizontalAlignment=HAlign_Center
    verticalAlignment=VAlign_Center
    background {
     bIsDynamicallyLoaded=false
     drawAs=NoDrawType
     imageType=NoImage
     mirroring=NoMirror
     resourceName=None
     resourceObject=null
     tiling=NoTile
     imageSize=<32 32>
     margin=<ltrb 0 0 0 0>
     outlineSettings {
      bUseBrushTransparency=false
      roundingType=HalfHeightRadius
      width=0
      color {colorUseRule=UseColor_Specified, specifiedColor=<rgba 0 0 0 0>}
      cornerRadii=<0 0 0 0>
     }
     tintColor {colorUseRule=UseColor_Specified, specifiedColor=<rgba 1 1 1 1>}
     uVRegion {bIsValid=false, max=<0 0>, min=<0 0>}
    }
    padding=<ltrb 64 24 64 24>
   }
   slot {class=OverlaySlot, properties {horizontalAlignment=HAlign_Center, verticalAlignment=VAlign_Center}}
  }
  HorizontalBox_0 {
   widgetClassPath=&HorizontalBox
   parent=Border_0
   widget=HorizontalBox_0
   slot {class=BorderSlot, properties {horizontalAlignment=HAlign_Center, verticalAlignment=VAlign_Center, padding=<ltrb 64 24 64 24>}}
  }
  InputActionWidget {
   widgetClassPath=&CommonActionWidget
   bInherited=true
   bIsVariable=true
   parent=HorizontalBox_0
   widget=InputActionWidget
   properties {
    progressMaterialParam=Percentage
    progressMaterialBrush {
     bIsDynamicallyLoaded=false
     drawAs=Image
     imageType=NoImage
     mirroring=NoMirror
     resourceName=None
     resourceObject=/Game/UI/Materials/HollowCircularProgressBar
     tiling=NoTile
     imageSize=<36 36>
     margin=<ltrb 0 0 0 0>
     outlineSettings {
      bUseBrushTransparency=false
      roundingType=HalfHeightRadius
      width=0
      color {colorUseRule=UseColor_Specified, specifiedColor=<rgba 0 0 0 0>}
      cornerRadii=<0 0 0 0>
     }
     tintColor {colorUseRule=UseColor_Specified, specifiedColor=<rgba 1 1 0 1>}
     uVRegion {bIsValid=false, max=<0 0>, min=<0 0>}
    }
   }
   slot {class=HorizontalBoxSlot, properties {horizontalAlignment=HAlign_Center, verticalAlignment=VAlign_Center, padding=<ltrb 0 0 4 0>}}
  }
  NamedSlot_ButtonContent {
   widgetClassPath=&NamedSlot
   bIsVariable=true
   parent=HorizontalBox_0
   widget=NamedSlot_ButtonContent
   properties {bExposeOnInstanceOnly=true}
   slot {class=HorizontalBoxSlot}
  }
 ]
}
