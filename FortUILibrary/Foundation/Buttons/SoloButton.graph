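# SoloButton is a CommonButtonLegacy widget: a styled button showing an input action icon beside an uppercase label on a material-driven backing.
# PreConstruct and Construct set the label, font size and padding, with optional overrides for size and padding.
# Enable and disable events swap the backing material and set a flag that suppresses the Hover animation, which tweens material parameters and text color.
# A Pressed animation exists, but its play node is unconnected.
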WidgetBlueprint /FortUILibrary/Foundation/Buttons/SoloButton
animations [
 Hover {
  binding_count=2
  length_seconds=0.15
  object=&self:Hover
  sequence {bindings [
    ButtonBacking {
     id=#1
     possessed_class=Image
     tracks [{
       type=MovieSceneWidgetMaterialTrack
       sections [{
         type=MovieSceneParameterSection
         channels [
          OnHover {type=MovieSceneScriptingFloatChannel, keys [[0,0],[3,1]]}
          RGB_1R {type=MovieSceneScriptingFloatChannel, keys [[0,0],[3,0.25]]}
          RGB_1G {type=MovieSceneScriptingFloatChannel, keys [[0,0.5457],[3,1]]}
          RGB_1B {type=MovieSceneScriptingFloatChannel, const [1,2]}
          RGB_1A {type=MovieSceneScriptingFloatChannel, const [1,2]}
          RGB_2R {type=MovieSceneScriptingFloatChannel, const [0,2]}
          RGB_2G {type=MovieSceneScriptingFloatChannel, keys [[0,0.2384],[3,0.45]]}
          RGB_2B {type=MovieSceneScriptingFloatChannel, keys [[0,0.7913],[3,1]]}
          RGB_2A {type=MovieSceneScriptingFloatChannel, const [1,1]}
         ]
        }]
      }]
    }
    Text_ButtonAction {
     id=#2
     possessed_class=CommonTextBlock
     tracks [{
       type=MovieSceneColorTrack
       display_name=Color and Opacity
       sections [{
         type=MovieSceneColorSection
         channels [
          Color.R {type=MovieSceneScriptingFloatChannel, keys [[0,0.0307],[3,1]]}
          Color.G {type=MovieSceneScriptingFloatChannel, keys [[0,0.8148],[3,1]]}
          Color.B {type=MovieSceneScriptingFloatChannel, keys [[0,0.9911],[3,1]]}
          Color.A {type=MovieSceneScriptingFloatChannel, const [1,2]}
         ]
        }]
      }]
    }
   ]}
 }
 Pressed {
  binding_count=2
  length_seconds=0.15
  object=&self:Pressed
  sequence {bindings [
    ButtonBacking {
     id=#3
     possessed_class=Image
     tracks [{
       type=MovieSceneWidgetMaterialTrack
       sections [{
         type=MovieSceneParameterSection
         channels [
          Pressed {type=MovieSceneScriptingFloatChannel, keys [[0,0],[2,1]]}
          Selected-RGB-Fade {type=MovieSceneScriptingFloatChannel, keys [[1,0],[3,1]]}
         ]
        }]
      }]
    }
    Border_FontColorChange {
     id=#4
     possessed_class=Border
     tracks [{
       type=MovieSceneColorTrack
       display_name=Content Color and Opacity
       sections [{
         type=MovieSceneColorSection
         channels [
          Color.R {type=MovieSceneScriptingFloatChannel, keys [[0,1],[3,0]]}
          Color.G {type=MovieSceneScriptingFloatChannel, keys [[0,1],[3,0.15]]}
          Color.B {type=MovieSceneScriptingFloatChannel, const [1,2]}
          Color.A {type=MovieSceneScriptingFloatChannel, const [1,2]}
         ]
        }]
      }]
    }
   ]}
 }
]
graphs [
 EventGraph {nodes [
   n1 = Event PreConstruct()
   HandleSize(execute=[@n1.then, @n2.then])
   > SetText!(execute, self:Self Object Reference, Text:Text=Get Button Description() -> (Button Description:Text)) -> (then)
   n2 = Event Construct()
   Event On Disabled() {is=title K2Node_Event}
   > Set isDisabled(isDisabled:bool=true) -> (Output_Get:bool)
   n3 = Get ButtonBacking() -> (ButtonBacking:Image Object Reference)
   > SetBrushFromMaterial(self=@n3, Material=/Game/UI/Foundation/Materials/Athena_Button_Solo_BlueBase_Disabled)
   Event On Enabled() {is=title K2Node_Event}
   > Set isDisabled(isDisabled=false)
   > SetBrushFromMaterial(self=@n3, Material=/Game/UI/Foundation/Materials/Athena_Button_Solo_BlueBase)
   Event On Hovered()
   n4 = Get isDisabled() -> (isDisabled:bool)
   > n5 = Branch(Condition=@n4)
   n6 = Get Hover() -> (Hover:Widget Animation Object Reference)
   PlayAnimationForward(execute=@n5.else, InAnimation=@n6, PlaybackSpeed=1.5)
   Event On Unhovered()
   > n7 = Branch(Condition=@n4)
   PlayAnimationReverse(execute=@n7.else, InAnimation=@n6)
   PlayAnimation(InAnimation=Get Pressed() -> (Pressed:Widget Animation Object Reference), PlayMode=PingPong)
   Event Tick()
   Event On Clicked()
  ]}
 HandleSize {nodes [
   HandleSize() {is=title K2Node_FunctionEntry}
   n1 = Select(Option 0:int=22, Option 1:int=Get FontSize() -> (FontSize:int=0), Index=Get FontSizeOveride() -> (FontSizeOveride:bool)) -> (ReturnValue:int)
   > SetFont(
    self=Get Text_ButtonAction() -> (Text_ButtonAction:Common Text Object Reference)
    InFontInfo=MakeSlateFontInfo!(
     FontObject:Object Reference=&BurbankSmall
     FontMaterial:Object Reference
     OutlineSettings:Font Outline Settings Structure=(OutlineSize=0,bMiteredCorners=False,bSeparateFillAlpha=False,bApplyOutlineToDropShadows=False,OutlineMaterial=None,OutlineColor=(R=0,G=0,B=0,A=1))
     TypefaceFontName:Name=BLACK
     Size:float="To Float (Integer)"(InInt=@n1)
     LetterSpacing:int=0
     SkewAmount:float=0
     bForceMonospaced:bool=false
     bMaterialIsStencil:bool=false
     MonospacedWidth:float=1
    ) -> (ReturnValue:Slate Font Info Structure)
   )
   > SetPadding!(
    execute
    self:Border Object Reference=Get Border_Container() -> (Border_Container:Border Object Reference)
    InPadding:Margin Structure=Select(
     Option 0:Margin Structure=Make Margin(Left=48, Top=14, Right=48, Bottom=14)
     Option 1:Margin Structure=Get Padding Overide() -> (Padding Overide:Margin Structure)
     Index=Get PaddingOveride() -> (PaddingOveride:bool)
    ) -> (ReturnValue:Margin Structure)
   ) -> (then)
  ]}
 SetText {nodes [
   n1 = SetText!() -> (then, Text:Text) {is=title K2Node_FunctionEntry}
   > n2 = Set Button Description(Button Description:Text=@n1) -> (Output_Get:Text)
   > "SetText (Text)"(self=Get Text_ButtonAction(), InText=@n2)
  ]}
]
variables [
 Button Description {type=string, default="{TEXT}"}
 FontSize {type=integer, default=22}
 FontSizeOveride {type=boolean}
 isDisabled {type=boolean}
 Padding Overide {type=Margin, default=<ltrb 48 12 48 12>}
 PaddingOveride {type=boolean}
]
widgets {
 info {inheritedWidgetCount=1, parentClass=&CommonButtonLegacy, rootWidgetClass=&SizeBox, widgetCount=8}
 widgets [
  SizeBox_Control {widgetClassPath=&SizeBox, bIsVariable=true, widget=SizeBox_Control, properties {widthOverride=256}}
  Overlay_29 {widgetClassPath=&Overlay, parent=SizeBox_Control, widget=Overlay_29, slot {class=SizeBoxSlot}}
  ButtonBacking {
   widgetClassPath=&Image
   bIsVariable=true
   parent=Overlay_29
   widget=ButtonBacking
   properties {brush {
     bIsDynamicallyLoaded=false
     drawAs=Image
     imageType=NoImage
     mirroring=NoMirror
     resourceName=None
     resourceObject=&Athena_Button_Solo_BlueBase
     tiling=NoTile
     imageSize=<32 32>
     margin=<ltrb 0 0 0 0>
     outlineSettings {
      bUseBrushTransparency=false
      roundingType=HalfHeightRadius
      width=0
      color {colorUseRule=UseColor_Specified, specifiedColor=<rgba 0 0 0 0>}
      cornerRadii=<0 0 0 0>
     }
     tintColor {colorUseRule=UseColor_Specified, specifiedColor=<rgba 1 1 1 1>}
     uVRegion {bIsValid=false, max=<0 0>, min=<0 0>}
    }}
   slot {class=OverlaySlot, properties {horizontalAlignment=HAlign_Fill, verticalAlignment=VAlign_Fill}}
  }
  Border_Container {
   widgetClassPath=&Border
   bIsVariable=true
   parent=Overlay_29
   widget=Border_Container
   properties {
    horizontalAlignment=HAlign_Center
    verticalAlignment=VAlign_Center
    background {
     bIsDynamicallyLoaded=false
     drawAs=NoDrawType
     imageType=NoImage
     mirroring=NoMirror
     resourceName=None
     resourceObject=null
     tiling=NoTile
     imageSize=<32 32>
     margin=<ltrb 0 0 0 0>
     outlineSettings {
      bUseBrushTransparency=false
      roundingType=HalfHeightRadius
      width=0
      color {colorUseRule=UseColor_Specified, specifiedColor=<rgba 0 0 0 0>}
      cornerRadii=<0 0 0 0>
     }
     tintColor {colorUseRule=UseColor_Specified, specifiedColor=<rgba 1 1 1 1>}
     uVRegion {bIsValid=false, max=<0 0>, min=<0 0>}
    }
    padding=<ltrb 0 0 0 0>
   }
   slot {class=OverlaySlot, properties {horizontalAlignment=HAlign_Fill, verticalAlignment=VAlign_Fill}}
  }
  HorizontalBox_0 {
   widgetClassPath=&HorizontalBox
   parent=Border_Container
   widget=HorizontalBox_0
   slot {class=BorderSlot, properties {horizontalAlignment=HAlign_Center, verticalAlignment=VAlign_Center, padding=<ltrb 0 0 0 0>}}
  }
  InputActionWidget {
   widgetClassPath=&CommonActionWidget
   bInherited=true
   bIsVariable=true
   parent=HorizontalBox_0
   widget=InputActionWidget
   properties {
    progressMaterialParam=Percentage
    progressMaterialBrush {
     bIsDynamicallyLoaded=false
     drawAs=Image
     imageType=NoImage
     mirroring=NoMirror
     resourceName=None
     resourceObject=/Game/UI/Materials/HollowCircularProgressBar
     tiling=NoTile
     imageSize=<36 36>
     margin=<ltrb 0 0 0 0>
     outlineSettings {
      bUseBrushTransparency=false
      roundingType=HalfHeightRadius
      width=0
      color {colorUseRule=UseColor_Specified, specifiedColor=<rgba 0 0 0 0>}
      cornerRadii=<0 0 0 0>
     }
     tintColor {colorUseRule=UseColor_Specified, specifiedColor=<rgba 1 1 0 1>}
     uVRegion {bIsValid=false, max=<0 0>, min=<0 0>}
    }
   }
   slot {class=HorizontalBoxSlot, properties {horizontalAlignment=HAlign_Center, verticalAlignment=VAlign_Center, padding=<ltrb 0 0 8 0>}}
  }
  Border_FontColorChange {
   widgetClassPath=&Border
   parent=HorizontalBox_0
   widget=Border_FontColorChange
   properties {
    horizontalAlignment=HAlign_Center
    background {
     bIsDynamicallyLoaded=false
     drawAs=NoDrawType
     imageType=NoImage
     mirroring=NoMirror
     resourceName=None
     resourceObject=null
     tiling=NoTile
     imageSize=<32 32>
     margin=<ltrb 0 0 0 0>
     outlineSettings {
      bUseBrushTransparency=false
      roundingType=HalfHeightRadius
      width=0
      color {colorUseRule=UseColor_Specified, specifiedColor=<rgba 0 0 0 0>}
      cornerRadii=<0 0 0 0>
     }
     tintColor {colorUseRule=UseColor_Specified, specifiedColor=<rgba 1 1 1 1>}
     uVRegion {bIsValid=false, max=<0 0>, min=<0 0>}
    }
    padding=<ltrb 0 0 0 0>
   }
   slot {class=HorizontalBoxSlot, properties {horizontalAlignment=HAlign_Center, verticalAlignment=VAlign_Center, size {sizeRule=Fill, value=1}}}
  }
  Text_ButtonAction {
   widgetClassPath=&CommonTextBlock
   bIsVariable=true
   parent=Border_FontColorChange
   widget=Text_ButtonAction
   properties {
    justification=Center
    scrollStyle=&TextScrollStyle-Base_C
    text=ACTION
    textTransformPolicy=ToUpper
    colorAndOpacity {colorUseRule=UseColor_Specified, specifiedColor=<rgba 0.0307 0.8148 0.9911 1>}
    font {
     bForceMonospaced=false
     bMaterialIsStencil=false
     fontMaterial=null
     fontObject=&BurbankSmall
     letterSpacing=0
     monospacedWidth=1
     size=16
     skewAmount=0
     typefaceFontName=Black
     outlineSettings {
      bApplyOutlineToDropShadows=false
      bMiteredCorners=false
      bSeparateFillAlpha=false
      outlineMaterial=null
      outlineSize=0
      outlineColor=<rgba 0 0 0 1>
     }
    }
    shadowOffset=<0 0>
   }
   slot {class=BorderSlot, properties {horizontalAlignment=HAlign_Center, padding=<ltrb 0 0 0 0>}}
  }
 ]
}
